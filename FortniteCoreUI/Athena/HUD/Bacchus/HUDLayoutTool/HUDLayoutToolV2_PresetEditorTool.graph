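# Preset Editor Tool panel for the mobile HUD Layout Tool V2, a developer-facing form for saving HUD layout presets and creating preset hotfixes.
# It takes preset container and preset tags, plus a toggle for saving to a preset extension and an initially collapsed extension container tag section.
# Save Preset, Create Hotfix and Close buttons sit below disclaimers about asset and DefaultGame.ini checkout.
# The event graph holds only unconnected stub events, with key controls inherited from the native FortMobileHUDLayoutToolV2_PresetEditorTool parent.
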WidgetBlueprint /FortniteCoreUI/Athena/HUD/Bacchus/HUDLayoutTool/HUDLayoutToolV2_PresetEditorTool
graphs [EventGraph {nodes [
   Event PreConstruct()
   Event Construct()
   Event Tick()
  ]}]
variables []
widgets {
 info {
  inheritedWidgetCount=8
  parentClass=/Script/FortniteUI.FortMobileHUDLayoutToolV2_PresetEditorTool
  rootWidgetClass=&SizeBox
  widgetCount=37
 }
 widgets [
  SizeBox_2 {
   widgetClassPath=&SizeBox
   widget=SizeBox_2
   properties {
    bOverride_HeightOverride=true
    bOverride_WidthOverride=true
    heightOverride=1000
    minDesiredHeight=1000
    minDesiredWidth=1100
    widthOverride=1300
   }
  }
  Overlay_21 {widgetClassPath=&Overlay, parent=SizeBox_2, widget=Overlay_21, slot {class=SizeBoxSlot}}
  Background {
   widgetClassPath=&Image
   parent=Overlay_21
   widget=Background
   properties {
    brush {
     bIsDynamicallyLoaded=false
     drawAs=Image
     imageType=NoImage
     mirroring=NoMirror
     resourceName=None
     resourceObject=null
     tiling=NoTile
     imageSize=<64 64>
     margin=<ltrb 0 0 0 0>
     outlineSettings {
      bUseBrushTransparency=false
      roundingType=HalfHeightRadius
      width=0
      color {colorUseRule=UseColor_Specified, specifiedColor=<rgba 0 0 0 0>}
      cornerRadii=<0 0 0 0>
     }
     tintColor {colorUseRule=UseColor_Specified, specifiedColor=<rgba 1 1 1 1>}
     uVRegion {bIsValid=false, max=<0 0>, min=<0 0>}
    }
    colorAndOpacity=<rgba 0.0235 0.0235 0.0235 1>
   }
   slot {class=OverlaySlot, properties {horizontalAlignment=HAlign_Fill, verticalAlignment=VAlign_Fill}}
  }
  VerticalBox_80 {
   widgetClassPath=&VerticalBox
   parent=Overlay_21
   widget=VerticalBox_80
   slot {class=OverlaySlot, properties {horizontalAlignment=HAlign_Fill}}
  }
  ToolTitle {
   widgetClassPath=&CommonTextBlock
   parent=VerticalBox_80
   widget=ToolTitle
   properties {
    text=Preset Editor Tool
    textTransformPolicy=ToUpper
    font {
     bForceMonospaced=false
     bMaterialIsStencil=false
     fontMaterial=null
     fontObject=&BurbankSmall
     letterSpacing=0
     monospacedWidth=1
     size=30
     skewAmount=0
     typefaceFontName=Black
     outlineSettings {
      bApplyOutlineToDropShadows=false
      bMiteredCorners=false
      bSeparateFillAlpha=false
      outlineMaterial=null
      outlineSize=0
      outlineColor=<rgba 0 0 0 1>
     }
    }
    shadowOffset=<0 0>
   }
   slot {class=VerticalBoxSlot, properties {horizontalAlignment=HAlign_Left, padding=<ltrb 20 20 0 10>}}
  }
  Separator_1 {
   widgetClassPath=&Border
   parent=VerticalBox_80
   widget=Separator_1
   properties {
    background {
     bIsDynamicallyLoaded=false
     drawAs=Image
     imageType=NoImage
     mirroring=NoMirror
     resourceName=None
     resourceObject=null
     tiling=NoTile
     imageSize=<32 32>
     margin=<ltrb 0 0 0 0>
     outlineSettings {
      bUseBrushTransparency=false
      roundingType=HalfHeightRadius
      width=0
      color {colorUseRule=UseColor_Specified, specifiedColor=<rgba 0 0 0 0>}
      cornerRadii=<0 0 0 0>
     }
     tintColor {colorUseRule=UseColor_Specified, specifiedColor=<rgba 1 1 1 0.3>}
     uVRegion {bIsValid=false, max=<0 0>, min=<0 0>}
    }
    padding=<ltrb 0 2 0 2>
   }
   slot {class=VerticalBoxSlot, properties {verticalAlignment=VAlign_Top, padding=<ltrb 20 0 20 10>}}
  }
  HorizontalBox_79 {
   widgetClassPath=&HorizontalBox
   parent=VerticalBox_80
   widget=HorizontalBox_79
   slot {class=VerticalBoxSlot, properties {verticalAlignment=VAlign_Center, padding=<ltrb 10 10 10 10>}}
  }
  PresetContainerTagTitle {
   widgetClassPath=&CommonTextBlock
   parent=HorizontalBox_79
   widget=PresetContainerTagTitle
   properties {
    text=Preset Container Tag :
    textTransformPolicy=ToUpper
    font {
     bForceMonospaced=false
     bMaterialIsStencil=false
     fontMaterial=null
     fontObject=&BurbankSmall
     letterSpacing=0
     monospacedWidth=1
     size=22
     skewAmount=0
     typefaceFontName=Black
     outlineSettings {
      bApplyOutlineToDropShadows=false
      bMiteredCorners=false
      bSeparateFillAlpha=false
      outlineMaterial=null
      outlineSize=0
      outlineColor=<rgba 0 0 0 1>
     }
    }
    shadowOffset=<0 0>
   }
   slot {class=HorizontalBoxSlot, properties {horizontalAlignment=HAlign_Left, padding=<ltrb 20 20 20 20>}}
  }
  SizeBox_0 {
   widgetClassPath=&SizeBox
   parent=HorizontalBox_79
   widget=SizeBox_0
   properties {bOverride_WidthOverride=true, widthOverride=380}
   slot {class=HorizontalBoxSlot, properties {padding=<ltrb 10 10 10 10>, size {sizeRule=Fill, value=1}}}
  }
  Border_0 {
   widgetClassPath=&Border
   parent=SizeBox_0
   widget=Border_0
   properties {verticalAlignment=VAlign_Center, brushColor=<rgba 0.6514 0.6514 0.6514 1>, padding=<ltrb 10 10 10 10>}
   slot {class=SizeBoxSlot}
  }
  EditableText_PresetContainerTag {
   widgetClassPath=/Script/FortniteUI.FortMobileEditableText
   bInherited=true
   parent=Border_0
   widget=EditableText_PresetContainerTag
   properties {
    hintText=HUDLayout.PresetContainer.ConatinerName
    text=HUDLayout.PresetContainer.Release
    widgetStyle {
     backgroundImageComposing {
      bIsDynamicallyLoaded=false
      drawAs=Border
      imageType=FullColor
      mirroring=NoMirror
      resourceName=../../../Engine/Content/Slate/Old/HyperlinkDotted.png
      resourceObject=null
      tiling=Both
      imageSize=<0 0>
      margin=<ltrb 0 0 0 0.1875>
      outlineSettings {
       bUseBrushTransparency=false
       roundingType=HalfHeightRadius
       width=0
       color {colorUseRule=UseColor_Specified, specifiedColor=<rgba 0 0 0 0>}
       cornerRadii=<0 0 0 0>
      }
      tintColor {colorUseRule=UseColor_Specified, specifiedColor=<rgba 1 1 1 1>}
      uVRegion {bIsValid=false, max=<0 0>, min=<0 0>}
     }
     backgroundImageSelected {
      bIsDynamicallyLoaded=false
      drawAs=Box
      imageType=FullColor
      mirroring=NoMirror
      resourceName=../../../Engine/Content/Slate/Common/EditableTextSelectionBackground.png
      resourceObject=null
      tiling=NoTile
      imageSize=<0 0>
      margin=<ltrb 0.25 0.25 0.25 0.25>
      outlineSettings {
       bUseBrushTransparency=false
       roundingType=HalfHeightRadius
       width=0
       color {colorUseRule=UseColor_Specified, specifiedColor=<rgba 0 0 0 0>}
       cornerRadii=<0 0 0 0>
      }
      tintColor {colorUseRule=UseColor_Specified, specifiedColor=<rgba 1 1 1 1>}
      uVRegion {bIsValid=false, max=<0 0>, min=<0 0>}
     }
     caretImage {
      bIsDynamicallyLoaded=false
      drawAs=Image
      imageType=FullColor
      mirroring=NoMirror
      resourceName=../../../Engine/Content/Slate/Old/White.png
      resourceObject=null
      tiling=NoTile
      imageSize=<16 16>
      margin=<ltrb 0 0 0 0>
      outlineSettings {
       bUseBrushTransparency=false
       roundingType=HalfHeightRadius
       width=0
       color {colorUseRule=UseColor_Specified, specifiedColor=<rgba 0 0 0 0>}
       cornerRadii=<0 0 0 0>
      }
      tintColor {colorUseRule=UseColor_Specified, specifiedColor=<rgba 1 1 1 1>}
      uVRegion {bIsValid=false, max=<0 0>, min=<0 0>}
     }
     colorAndOpacity {colorUseRule=UseColor_Specified, specifiedColor=<rgba 0 0 0 1>}
     font {
      bForceMonospaced=false
      bMaterialIsStencil=false
      fontMaterial=null
      fontObject=&BurbankSmall
      letterSpacing=0
      monospacedWidth=1
      size=20
      skewAmount=0
      typefaceFontName=Bold
      outlineSettings {
       bApplyOutlineToDropShadows=false
       bMiteredCorners=false
       bSeparateFillAlpha=false
       outlineMaterial=null
       outlineSize=0
       outlineColor=<rgba 0 0 0 1>
      }
     }
     hintColorAndOpacity {colorUseRule=UseColor_Foreground, specifiedColor=<rgba 1 0 1 1>}
     selectedBackgroundColor {colorUseRule=UseColor_Foreground, specifiedColor=<rgba 1 0 1 1>}
    }
   }
   slot {class=BorderSlot, properties {verticalAlignment=VAlign_Center, padding=<ltrb 10 10 10 10>}}
  }
  HorizontalBox {
   widgetClassPath=&HorizontalBox
   parent=VerticalBox_80
   widget=HorizontalBox
   slot {class=VerticalBoxSlot, properties {verticalAlignment=VAlign_Center, padding=<ltrb 10 10 10 10>}}
  }
  PresetTagTitle {
   widgetClassPath=&CommonTextBlock
   parent=HorizontalBox
   widget=PresetTagTitle
   properties {
    text=Preset Tag :
    textTransformPolicy=ToUpper
    font {
     bForceMonospaced=false
     bMaterialIsStencil=false
     fontMaterial=null
     fontObject=&BurbankSmall
     letterSpacing=0
     monospacedWidth=1
     size=22
     skewAmount=0
     typefaceFontName=Black
     outlineSettings {
      bApplyOutlineToDropShadows=false
      bMiteredCorners=false
      bSeparateFillAlpha=false
      outlineMaterial=null
      outlineSize=0
      outlineColor=<rgba 0 0 0 1>
     }
    }
    shadowOffset=<0 0>
   }
   slot {class=HorizontalBoxSlot, properties {horizontalAlignment=HAlign_Left, padding=<ltrb 20 20 20 20>}}
  }
  SizeBox {
   widgetClassPath=&SizeBox
   parent=HorizontalBox
   widget=SizeBox
   properties {bOverride_WidthOverride=true, widthOverride=380}
   slot {class=HorizontalBoxSlot, properties {padding=<ltrb 10 10 10 10>, size {sizeRule=Fill, value=1}}}
  }
  Border {
   widgetClassPath=&Border
   parent=SizeBox
   widget=Border
   properties {verticalAlignment=VAlign_Center, brushColor=<rgba 0.6514 0.6514 0.6514 1>, padding=<ltrb 10 10 10 10>}
   slot {class=SizeBoxSlot}
  }
  EditableText_PresetTag {
   widgetClassPath=&FortMobileEditableText
   bInherited=true
   parent=Border
   widget=EditableText_PresetTag
   properties {
    hintText=HUDLayout.Preset.PresetName
    text=HUDLayout.Preset.
    widgetStyle {
     backgroundImageComposing {
      bIsDynamicallyLoaded=false
      drawAs=Border
      imageType=FullColor
      mirroring=NoMirror
      resourceName=../../../Engine/Content/Slate/Old/HyperlinkDotted.png
      resourceObject=null
      tiling=Both
      imageSize=<0 0>
      margin=<ltrb 0 0 0 0.1875>
      outlineSettings {
       bUseBrushTransparency=false
       roundingType=HalfHeightRadius
       width=0
       color {colorUseRule=UseColor_Specified, specifiedColor=<rgba 0 0 0 0>}
       cornerRadii=<0 0 0 0>
      }
      tintColor {colorUseRule=UseColor_Specified, specifiedColor=<rgba 1 1 1 1>}
      uVRegion {bIsValid=false, max=<0 0>, min=<0 0>}
     }
     backgroundImageSelected {
      bIsDynamicallyLoaded=false
      drawAs=Box
      imageType=FullColor
      mirroring=NoMirror
      resourceName=../../../Engine/Content/Slate/Common/EditableTextSelectionBackground.png
      resourceObject=null
      tiling=NoTile
      imageSize=<0 0>
      margin=<ltrb 0.25 0.25 0.25 0.25>
      outlineSettings {
       bUseBrushTransparency=false
       roundingType=HalfHeightRadius
       width=0
       color {colorUseRule=UseColor_Specified, specifiedColor=<rgba 0 0 0 0>}
       cornerRadii=<0 0 0 0>
      }
      tintColor {colorUseRule=UseColor_Specified, specifiedColor=<rgba 1 1 1 1>}
      uVRegion {bIsValid=false, max=<0 0>, min=<0 0>}
     }
     caretImage {
      bIsDynamicallyLoaded=false
      drawAs=Image
      imageType=FullColor
      mirroring=NoMirror
      resourceName=../../../Engine/Content/Slate/Old/White.png
      resourceObject=null
      tiling=NoTile
      imageSize=<16 16>
      margin=<ltrb 0 0 0 0>
      outlineSettings {
       bUseBrushTransparency=false
       roundingType=HalfHeightRadius
       width=0
       color {colorUseRule=UseColor_Specified, specifiedColor=<rgba 0 0 0 0>}
       cornerRadii=<0 0 0 0>
      }
      tintColor {colorUseRule=UseColor_Specified, specifiedColor=<rgba 1 1 1 1>}
      uVRegion {bIsValid=false, max=<0 0>, min=<0 0>}
     }
     colorAndOpacity {colorUseRule=UseColor_Specified, specifiedColor=<rgba 0 0 0 1>}
     font {
      bForceMonospaced=false
      bMaterialIsStencil=false
      fontMaterial=null
      fontObject=&BurbankSmall
      letterSpacing=0
      monospacedWidth=1
      size=20
      skewAmount=0
      typefaceFontName=Bold
      outlineSettings {
       bApplyOutlineToDropShadows=false
       bMiteredCorners=false
       bSeparateFillAlpha=false
       outlineMaterial=null
       outlineSize=0
       outlineColor=<rgba 0 0 0 1>
      }
     }
     hintColorAndOpacity {colorUseRule=UseColor_Foreground, specifiedColor=<rgba 1 0 1 1>}
     selectedBackgroundColor {colorUseRule=UseColor_Foreground, specifiedColor=<rgba 1 0 1 1>}
    }
   }
   slot {class=BorderSlot, properties {verticalAlignment=VAlign_Center, padding=<ltrb 10 10 10 10>}}
  }
  Separator {
   widgetClassPath=&Border
   parent=VerticalBox_80
   widget=Separator
   properties {
    background {
     bIsDynamicallyLoaded=false
     drawAs=Image
     imageType=NoImage
     mirroring=NoMirror
     resourceName=None
     resourceObject=null
     tiling=NoTile
     imageSize=<32 32>
     margin=<ltrb 0 0 0 0>
     outlineSettings {
      bUseBrushTransparency=false
      roundingType=HalfHeightRadius
      width=0
      color {colorUseRule=UseColor_Specified, specifiedColor=<rgba 0 0 0 0>}
      cornerRadii=<0 0 0 0>
     }
     tintColor {colorUseRule=UseColor_Specified, specifiedColor=<rgba 1 1 1 0.3>}
     uVRegion {bIsValid=false, max=<0 0>, min=<0 0>}
    }
    padding=<ltrb 0 2 0 2>
   }
   slot {class=VerticalBoxSlot, properties {verticalAlignment=VAlign_Top, padding=<ltrb 20 10 20 10>}}
  }
  HorizontalBox_418 {widgetClassPath=&HorizontalBox, parent=VerticalBox_80, widget=HorizontalBox_418, slot {class=VerticalBoxSlot}}
  Overlay_156 {widgetClassPath=&Overlay, parent=HorizontalBox_418, widget=Overlay_156, slot {class=HorizontalBoxSlot}}
  CheckboxBorder {
   widgetClassPath=&Image
   parent=Overlay_156
   widget=CheckboxBorder
   properties {
    brush {
     bIsDynamicallyLoaded=false
     drawAs=Image
     imageType=NoImage
     mirroring=NoMirror
     resourceName=None
     resourceObject=/Game/Athena/HUD/Bacchus/HUDLayoutTool/HLT_Icons/M_HUDLayoutToolV2_TiltSquare_Inst
     tiling=NoTile
     imageSize=<58 58>
     margin=<ltrb 0 0 0 0>
     outlineSettings {
      bUseBrushTransparency=false
      roundingType=HalfHeightRadius
      width=0
      color {colorUseRule=UseColor_Specified, specifiedColor=<rgba 0 0 0 0>}
      cornerRadii=<0 0 0 0>
     }
     tintColor {colorUseRule=UseColor_Specified, specifiedColor=<rgba 1 1 1 1>}
     uVRegion {bIsValid=false, max=<0 0>, min=<0 0>}
    }
    colorAndOpacity=<rgba 0 0.117 1 1>
   }
   slot {class=OverlaySlot, properties {horizontalAlignment=HAlign_Center, verticalAlignment=VAlign_Center}}
  }
  Button_ToggleSavingToExtension {
   widgetClassPath=/Game/Athena/HUD/Bacchus/HUDLayoutTool/HUDLayoutToolV2_Button
   bInherited=true
   parent=Overlay_156
   widget=Button_ToggleSavingToExtension
   properties {
    bIsToggleable=true
    bOverride_HeightOverride=true
    bOverride_WidthOverride=true
    buttonHeightOverride=60
    buttonWidthOverride=60
    text=""
    backgroundBrush {
     bIsDynamicallyLoaded=false
     drawAs=NoDrawType
     imageType=NoImage
     mirroring=NoMirror
     resourceName=None
     resourceObject=null
     tiling=NoTile
     imageSize=<32 32>
     margin=<ltrb 0 0 0 0>
     outlineSettings {
      bUseBrushTransparency=false
      roundingType=HalfHeightRadius
      width=0
      color {colorUseRule=UseColor_Specified, specifiedColor=<rgba 0 0 0 0>}
      cornerRadii=<0 0 0 0>
     }
     tintColor {colorUseRule=UseColor_Specified, specifiedColor=<rgba 1 1 1 1>}
     uVRegion {bIsValid=false, max=<0 0>, min=<0 0>}
    }
    iconBrush {
     bIsDynamicallyLoaded=false
     drawAs=Image
     imageType=NoImage
     mirroring=NoMirror
     resourceName=None
     resourceObject=/Game/Athena/HUD/Bacchus/HUDLayoutTool/HLT_Icons/T-HLT-CheckMark-32_Sprite
     tiling=NoTile
     imageSize=<37 36>
     margin=<ltrb 0 0 0 0>
     outlineSettings {
      bUseBrushTransparency=false
      roundingType=HalfHeightRadius
      width=0
      color {colorUseRule=UseColor_Specified, specifiedColor=<rgba 0 0 0 0>}
      cornerRadii=<0 0 0 0>
     }
     tintColor {colorUseRule=UseColor_Specified, specifiedColor=<rgba 1 1 1 1>}
     uVRegion {bIsValid=false, max=<0 0>, min=<0 0>}
    }
    internalPadding=<ltrb 0 0 0 0>
    normalButtonStyle {
     iconAngle=0
     backgroundColor=<rgba 0 0.117 1 1>
     iconColor=<rgba 1 1 1 0>
     textColor {colorUseRule=UseColor_Specified, specifiedColor=<rgba 1 1 1 1>}
    }
    selectedButtonStyle {
     iconAngle=0
     backgroundColor=<rgba 0 0.117 1 1>
     iconColor=<rgba 1 1 1 1>
     textColor {colorUseRule=UseColor_Specified, specifiedColor=<rgba 1 1 1 1>}
    }
   }
   slot {class=OverlaySlot, properties {horizontalAlignment=HAlign_Center, verticalAlignment=VAlign_Center, padding=<ltrb 20 20 20 20>}}
  }
  SavingToPresetExtensionTitle {
   widgetClassPath=&CommonTextBlock
   parent=HorizontalBox_418
   widget=SavingToPresetExtensionTitle
   properties {
    text=Saving to Preset Extension
    textTransformPolicy=ToUpper
    font {
     bForceMonospaced=false
     bMaterialIsStencil=false
     fontMaterial=null
     fontObject=&BurbankSmall
     letterSpacing=0
     monospacedWidth=1
     size=22
     skewAmount=0
     typefaceFontName=Black
     outlineSettings {
      bApplyOutlineToDropShadows=false
      bMiteredCorners=false
      bSeparateFillAlpha=false
      outlineMaterial=null
      outlineSize=0
      outlineColor=<rgba 0 0 0 1>
     }
    }
    shadowOffset=<0 0>
   }
   slot {class=HorizontalBoxSlot, properties {horizontalAlignment=HAlign_Left, verticalAlignment=VAlign_Center, padding=<ltrb 20 20 20 20>}}
  }
  VerticalBox_ExtensionSection {
   widgetClassPath=&VerticalBox
   bInherited=true
   parent=VerticalBox_80
   widget=VerticalBox_ExtensionSection
   properties {visibility=Collapsed}
   slot {class=VerticalBoxSlot, properties {verticalAlignment=VAlign_Center, padding=<ltrb 10 10 10 10>}}
  }
  HorizontalBox_1 {widgetClassPath=&HorizontalBox, parent=VerticalBox_ExtensionSection, widget=HorizontalBox_1, slot {class=VerticalBoxSlot}}
  PresetExtensionContainerTagTitle {
   widgetClassPath=&CommonTextBlock
   parent=HorizontalBox_1
   widget=PresetExtensionContainerTagTitle
   properties {
    text=Preset Extension Container Tag :
    textTransformPolicy=ToUpper
    font {
     bForceMonospaced=false
     bMaterialIsStencil=false
     fontMaterial=null
     fontObject=&BurbankSmall
     letterSpacing=0
     monospacedWidth=1
     size=22
     skewAmount=0
     typefaceFontName=Black
     outlineSettings {
      bApplyOutlineToDropShadows=false
      bMiteredCorners=false
      bSeparateFillAlpha=false
      outlineMaterial=null
      outlineSize=0
      outlineColor=<rgba 0 0 0 1>
     }
    }
    shadowOffset=<0 0>
   }
   slot {class=HorizontalBoxSlot, properties {horizontalAlignment=HAlign_Left, padding=<ltrb 20 20 20 20>}}
  }
  SizeBox_1 {
   widgetClassPath=&SizeBox
   parent=HorizontalBox_1
   widget=SizeBox_1
   properties {bOverride_WidthOverride=true, widthOverride=380}
   slot {class=HorizontalBoxSlot, properties {padding=<ltrb 10 10 10 10>, size {sizeRule=Fill, value=1}}}
  }
  Border_1 {
   widgetClassPath=&Border
   parent=SizeBox_1
   widget=Border_1
   properties {verticalAlignment=VAlign_Center, brushColor=<rgba 0.6514 0.6514 0.6514 1>, padding=<ltrb 10 10 10 10>}
   slot {class=SizeBoxSlot}
  }
  EditableText_PresetExtensionContainerTag {
   widgetClassPath=&FortMobileEditableText
   bInherited=true
   parent=Border_1
   widget=EditableText_PresetExtensionContainerTag
   properties {
    hintText=HUDLayout.PresetExtension.PluginName
    text=HUDLayout.PresetExtensionContainer.
    widgetStyle {
     backgroundImageComposing {
      bIsDynamicallyLoaded=false
      drawAs=Border
      imageType=FullColor
      mirroring=NoMirror
      resourceName=../../../Engine/Content/Slate/Old/HyperlinkDotted.png
      resourceObject=null
      tiling=Both
      imageSize=<0 0>
      margin=<ltrb 0 0 0 0.1875>
      outlineSettings {
       bUseBrushTransparency=false
       roundingType=HalfHeightRadius
       width=0
       color {colorUseRule=UseColor_Specified, specifiedColor=<rgba 0 0 0 0>}
       cornerRadii=<0 0 0 0>
      }
      tintColor {colorUseRule=UseColor_Specified, specifiedColor=<rgba 1 1 1 1>}
      uVRegion {bIsValid=false, max=<0 0>, min=<0 0>}
     }
     backgroundImageSelected {
      bIsDynamicallyLoaded=false
      drawAs=Box
      imageType=FullColor
      mirroring=NoMirror
      resourceName=../../../Engine/Content/Slate/Common/EditableTextSelectionBackground.png
      resourceObject=null
      tiling=NoTile
      imageSize=<0 0>
      margin=<ltrb 0.25 0.25 0.25 0.25>
      outlineSettings {
       bUseBrushTransparency=false
       roundingType=HalfHeightRadius
       width=0
       color {colorUseRule=UseColor_Specified, specifiedColor=<rgba 0 0 0 0>}
       cornerRadii=<0 0 0 0>
      }
      tintColor {colorUseRule=UseColor_Specified, specifiedColor=<rgba 1 1 1 1>}
      uVRegion {bIsValid=false, max=<0 0>, min=<0 0>}
     }
     caretImage {
      bIsDynamicallyLoaded=false
      drawAs=Image
      imageType=FullColor
      mirroring=NoMirror
      resourceName=../../../Engine/Content/Slate/Old/White.png
      resourceObject=null
      tiling=NoTile
      imageSize=<16 16>
      margin=<ltrb 0 0 0 0>
      outlineSettings {
       bUseBrushTransparency=false
       roundingType=HalfHeightRadius
       width=0
       color {colorUseRule=UseColor_Specified, specifiedColor=<rgba 0 0 0 0>}
       cornerRadii=<0 0 0 0>
      }
      tintColor {colorUseRule=UseColor_Specified, specifiedColor=<rgba 1 1 1 1>}
      uVRegion {bIsValid=false, max=<0 0>, min=<0 0>}
     }
     colorAndOpacity {colorUseRule=UseColor_Specified, specifiedColor=<rgba 0 0 0 1>}
     font {
      bForceMonospaced=false
      bMaterialIsStencil=false
      fontMaterial=null
      fontObject=&BurbankSmall
      letterSpacing=0
      monospacedWidth=1
      size=20
      skewAmount=0
      typefaceFontName=Bold
      outlineSettings {
       bApplyOutlineToDropShadows=false
       bMiteredCorners=false
       bSeparateFillAlpha=false
       outlineMaterial=null
       outlineSize=0
       outlineColor=<rgba 0 0 0 1>
      }
     }
     hintColorAndOpacity {colorUseRule=UseColor_Foreground, specifiedColor=<rgba 1 0 1 1>}
     selectedBackgroundColor {colorUseRule=UseColor_Foreground, specifiedColor=<rgba 1 0 1 1>}
    }
   }
   slot {class=BorderSlot, properties {verticalAlignment=VAlign_Center, padding=<ltrb 10 10 10 10>}}
  }
  DisclaimerTitle {
   widgetClassPath=&CommonTextBlock
   parent=VerticalBox_80
   widget=DisclaimerTitle
   properties {
    autoWrapText=true
    text=*Disclaimer*
    colorAndOpacity {colorUseRule=UseColor_Specified, specifiedColor=<rgba 0.3073 0.3073 0.3073 1>}
    font {
     bForceMonospaced=false
     bMaterialIsStencil=false
     fontMaterial=null
     fontObject=&BurbankSmall
     letterSpacing=0
     monospacedWidth=1
     size=22
     skewAmount=0
     typefaceFontName=Bold
     outlineSettings {
      bApplyOutlineToDropShadows=false
      bMiteredCorners=false
      bSeparateFillAlpha=false
      outlineMaterial=null
      outlineSize=0
      outlineColor=<rgba 0 0 0 1>
     }
    }
    shadowOffset=<0 0>
   }
   slot {class=VerticalBoxSlot, properties {padding=<ltrb 20 20 20 0>}}
  }
  DisclaimerText {
   widgetClassPath=&CommonTextBlock
   parent=VerticalBox_80
   widget=DisclaimerText
   properties {
    autoWrapText=true
    text="- Don't forget to have the appropriate Preset Files checked out. E.g : MobileHUDPresets.uasset, YourPresetExtension.uasset"
    colorAndOpacity {colorUseRule=UseColor_Specified, specifiedColor=<rgba 0.3073 0.3073 0.3073 1>}
    font {
     bForceMonospaced=false
     bMaterialIsStencil=false
     fontMaterial=null
     fontObject=&BurbankSmall
     letterSpacing=0
     monospacedWidth=1
     size=22
     skewAmount=0
     typefaceFontName=Bold
     outlineSettings {
      bApplyOutlineToDropShadows=false
      bMiteredCorners=false
      bSeparateFillAlpha=false
      outlineMaterial=null
      outlineSize=0
      outlineColor=<rgba 0 0 0 1>
     }
    }
    shadowOffset=<0 0>
   }
   slot {class=VerticalBoxSlot, properties {padding=<ltrb 20 20 20 0>}}
  }
  DisclaimerText_1 {
   widgetClassPath=&CommonTextBlock
   parent=VerticalBox_80
   widget=DisclaimerText_1
   properties {
    autoWrapText=true
    text="- If you're creating a new preset. Please, manually name it in the preset container asset. E.g : MobileHUDPresets.uasset"
    colorAndOpacity {colorUseRule=UseColor_Specified, specifiedColor=<rgba 0.3073 0.3073 0.3073 1>}
    font {
     bForceMonospaced=false
     bMaterialIsStencil=false
     fontMaterial=null
     fontObject=&BurbankSmall
     letterSpacing=0
     monospacedWidth=1
     size=22
     skewAmount=0
     typefaceFontName=Bold
     outlineSettings {
      bApplyOutlineToDropShadows=false
      bMiteredCorners=false
      bSeparateFillAlpha=false
      outlineMaterial=null
      outlineSize=0
      outlineColor=<rgba 0 0 0 1>
     }
    }
    shadowOffset=<0 0>
   }
   slot {class=VerticalBoxSlot, properties {padding=<ltrb 20 20 20 0>}}
  }
  DisclaimerText_2 {
   widgetClassPath=&CommonTextBlock
   parent=VerticalBox_80
   widget=DisclaimerText_2
   properties {
    autoWrapText=true
    text=- If you're adding a Hotfix for a Preset. Make sure you have DefaultGame.ini checked out.
    colorAndOpacity {colorUseRule=UseColor_Specified, specifiedColor=<rgba 0.3073 0.3073 0.3073 1>}
    font {
     bForceMonospaced=false
     bMaterialIsStencil=false
     fontMaterial=null
     fontObject=&BurbankSmall
     letterSpacing=0
     monospacedWidth=1
     size=22
     skewAmount=0
     typefaceFontName=Bold
     outlineSettings {
      bApplyOutlineToDropShadows=false
      bMiteredCorners=false
      bSeparateFillAlpha=false
      outlineMaterial=null
      outlineSize=0
      outlineColor=<rgba 0 0 0 1>
     }
    }
    shadowOffset=<0 0>
   }
   slot {class=VerticalBoxSlot, properties {padding=<ltrb 20 20 20 0>}}
  }
  DisclaimerText_3 {
   widgetClassPath=&CommonTextBlock
   parent=VerticalBox_80
   widget=DisclaimerText_3
   properties {
    autoWrapText=true
    text=- If there are any Preset Hotfixes pending please resolve them before editing any Presets or Extensions.
    colorAndOpacity {colorUseRule=UseColor_Specified, specifiedColor=<rgba 0.3073 0.3073 0.3073 1>}
    font {
     bForceMonospaced=false
     bMaterialIsStencil=false
     fontMaterial=null
     fontObject=&BurbankSmall
     letterSpacing=0
     monospacedWidth=1
     size=22
     skewAmount=0
     typefaceFontName=Bold
     outlineSettings {
      bApplyOutlineToDropShadows=false
      bMiteredCorners=false
      bSeparateFillAlpha=false
      outlineMaterial=null
      outlineSize=0
      outlineColor=<rgba 0 0 0 1>
     }
    }
    shadowOffset=<0 0>
   }
   slot {class=VerticalBoxSlot, properties {padding=<ltrb 20 20 20 0>}}
  }
  HorizontalBox_523 {
   widgetClassPath=&HorizontalBox
   parent=Overlay_21
   widget=HorizontalBox_523
   slot {class=OverlaySlot, properties {horizontalAlignment=HAlign_Right, verticalAlignment=VAlign_Bottom}}
  }
  Button_SavePreset {
   widgetClassPath=&HUDLayoutToolV2_Button
   bInherited=true
   bIsVariable=true
   parent=HorizontalBox_523
   widget=Button_SavePreset
   properties {
    text=Save Preset
    visibility=Visible
    iconBrush {
     bIsDynamicallyLoaded=false
     drawAs=NoDrawType
     imageType=NoImage
     mirroring=NoMirror
     resourceName=None
     resourceObject=null
     tiling=NoTile
     imageSize=<51 37>
     margin=<ltrb 0 0 0 0>
     outlineSettings {
      bUseBrushTransparency=false
      roundingType=HalfHeightRadius
      width=0
      color {colorUseRule=UseColor_Specified, specifiedColor=<rgba 0 0 0 0>}
      cornerRadii=<0 0 0 0>
     }
     tintColor {colorUseRule=UseColor_Specified, specifiedColor=<rgba 1 1 1 1>}
     uVRegion {bIsValid=false, max=<0 0>, min=<0 0>}
    }
    internalPadding=<ltrb 12 12 12 12>
    selectedButtonStyle {
     iconAngle=0
     backgroundColor=<rgba 0.0232 0.0232 0.0232 1>
     iconColor=<rgba 1 1 1 1>
     textColor {colorUseRule=UseColor_Specified, specifiedColor=<rgba 1 1 1 1>}
    }
   }
   slot {class=HorizontalBoxSlot, properties {horizontalAlignment=HAlign_Right, padding=<ltrb 20 20 20 20>}}
  }
  Button_CreateHotfix {
   widgetClassPath=&HUDLayoutToolV2_Button
   bInherited=true
   bIsVariable=true
   parent=HorizontalBox_523
   widget=Button_CreateHotfix
   properties {
    text=Create Hotfix
    visibility=Visible
    iconBrush {
     bIsDynamicallyLoaded=false
     drawAs=NoDrawType
     imageType=NoImage
     mirroring=NoMirror
     resourceName=None
     resourceObject=null
     tiling=NoTile
     imageSize=<51 37>
     margin=<ltrb 0 0 0 0>
     outlineSettings {
      bUseBrushTransparency=false
      roundingType=HalfHeightRadius
      width=0
      color {colorUseRule=UseColor_Specified, specifiedColor=<rgba 0 0 0 0>}
      cornerRadii=<0 0 0 0>
     }
     tintColor {colorUseRule=UseColor_Specified, specifiedColor=<rgba 1 1 1 1>}
     uVRegion {bIsValid=false, max=<0 0>, min=<0 0>}
    }
    internalPadding=<ltrb 12 12 12 12>
    selectedButtonStyle {
     iconAngle=0
     backgroundColor=<rgba 0.0232 0.0232 0.0232 1>
     iconColor=<rgba 1 1 1 1>
     textColor {colorUseRule=UseColor_Specified, specifiedColor=<rgba 1 1 1 1>}
    }
   }
   slot {class=HorizontalBoxSlot, properties {horizontalAlignment=HAlign_Right, padding=<ltrb 20 20 20 20>}}
  }
  Button_Close {
   widgetClassPath=&HUDLayoutToolV2_Button
   bInherited=true
   bIsVariable=true
   parent=HorizontalBox_523
   widget=Button_Close
   properties {
    text=Close
    visibility=Visible
    iconBrush {
     bIsDynamicallyLoaded=false
     drawAs=NoDrawType
     imageType=NoImage
     mirroring=NoMirror
     resourceName=None
     resourceObject=null
     tiling=NoTile
     imageSize=<51 37>
     margin=<ltrb 0 0 0 0>
     outlineSettings {
      bUseBrushTransparency=false
      roundingType=HalfHeightRadius
      width=0
      color {colorUseRule=UseColor_Specified, specifiedColor=<rgba 0 0 0 0>}
      cornerRadii=<0 0 0 0>
     }
     tintColor {colorUseRule=UseColor_Specified, specifiedColor=<rgba 1 1 1 1>}
     uVRegion {bIsValid=false, max=<0 0>, min=<0 0>}
    }
    internalPadding=<ltrb 12 12 12 12>
    normalButtonStyle {
     iconAngle=0
     backgroundColor=<rgba 0.6939 0.0137 0.016 1>
     iconColor=<rgba 1 1 1 1>
     textColor {colorUseRule=UseColor_Specified, specifiedColor=<rgba 1 1 1 1>}
    }
    pressedButtonStyle {
     iconAngle=0
     backgroundColor=<rgba 0.7529 0.0723 0.0762 1>
     iconColor=<rgba 1 1 1 1>
     textColor {colorUseRule=UseColor_Specified, specifiedColor=<rgba 1 1 1 1>}
    }
   }
   slot {class=HorizontalBoxSlot, properties {horizontalAlignment=HAlign_Right, padding=<ltrb 20 20 20 20>}}
  }
 ]
}
