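# Tutorial text prompt widget for the Bacchus HUD Layout Tool V2, showing an uppercase title and a wrapped description.
# The layout layers a drop-shadow sprite and a red-tinted textbox material behind the text, using negative padding to overhang the backgrounds and title.
# A SizeBox caps the description width.
# Both text blocks are inherited from the native HUDLayoutToolV2_TutorialTextPrompt parent; the event graph holds only empty events.
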WidgetBlueprint /FortniteCoreUI/Athena/HUD/Bacchus/HUDLayoutTool/HUDLayoutToolV2_TutorialTextPrompt
graphs [EventGraph {nodes [
   Event PreConstruct()
   Event Construct()
   Event Tick()
  ]}]
variables []
widgets {
 info {
  inheritedWidgetCount=2
  parentClass=/Script/FortniteUI.HUDLayoutToolV2_TutorialTextPrompt
  rootWidgetClass=&ScaleBox
  widgetCount=9
 }
 widgets [
  ScaleBox_1 {widgetClassPath=&ScaleBox, widget=ScaleBox_1, properties {stretch=None}}
  OverlayBG {
   widgetClassPath=&Overlay
   parent=ScaleBox_1
   widget=OverlayBG
   slot {class=ScaleBoxSlot, properties {horizontalAlignment=HAlign_Fill, verticalAlignment=VAlign_Fill}}
  }
  Image_BGDropShadow {
   widgetClassPath=&Image
   parent=OverlayBG
   widget=Image_BGDropShadow
   properties {
    visibility=SelfHitTestInvisible
    brush {
     bIsDynamicallyLoaded=false
     drawAs=Image
     imageType=NoImage
     mirroring=NoMirror
     resourceName=None
     resourceObject=/Game/Athena/HUD/Bacchus/HUDLayoutTool/HLT_Icons/T-HLT-TutorialShadow_Sprite
     tiling=NoTile
     imageSize=<64 64>
     margin=<ltrb 0.5 0.5 0.6881 0.5>
     outlineSettings {
      bUseBrushTransparency=false
      roundingType=HalfHeightRadius
      width=0
      color {colorUseRule=UseColor_Specified, specifiedColor=<rgba 0 0 0 0>}
      cornerRadii=<0 0 0 0>
     }
     tintColor {colorUseRule=UseColor_Specified, specifiedColor=<rgba 1 1 1 1>}
     uVRegion {bIsValid=false, max=<0 0>, min=<0 0>}
    }
    colorAndOpacity=<rgba 0 0 0 0.8>
   }
   slot {class=OverlaySlot, properties {horizontalAlignment=HAlign_Fill, verticalAlignment=VAlign_Fill, padding=<ltrb -50 -70 -70 -60>}}
  }
  Image_BGInset {
   widgetClassPath=&Image
   parent=OverlayBG
   widget=Image_BGInset
   properties {
    brush {
     bIsDynamicallyLoaded=false
     drawAs=Image
     imageType=NoImage
     mirroring=NoMirror
     resourceName=None
     resourceObject=/Game/Athena/HUD/Bacchus/HUDLayoutTool/HLT_Icons/M_HUDLayoutToolV2_TutorialTextbox
     tiling=NoTile
     imageSize=<32 32>
     margin=<ltrb 0.15 0.2 0.15 0.13>
     outlineSettings {
      bUseBrushTransparency=false
      roundingType=HalfHeightRadius
      width=0
      color {colorUseRule=UseColor_Specified, specifiedColor=<rgba 0 0 0 0>}
      cornerRadii=<0 0 0 0>
     }
     tintColor {colorUseRule=UseColor_Specified, specifiedColor=<rgba 1 1 1 1>}
     uVRegion {bIsValid=false, max=<0 0>, min=<0 0>}
    }
    colorAndOpacity=<rgba 0.807 0.0027 0.0027 1>
   }
   slot {class=OverlaySlot, properties {horizontalAlignment=HAlign_Fill, verticalAlignment=VAlign_Fill, padding=<ltrb -12 -40 -20 -20>}}
  }
  Overlay_1 {widgetClassPath=&Overlay, parent=OverlayBG, widget=Overlay_1, slot {class=OverlaySlot}}
  TextBlock_Title {
   widgetClassPath=&CommonTextBlock
   bInherited=true
   parent=Overlay_1
   widget=TextBlock_Title
   properties {
    text="Properties!"
    textTransformPolicy=ToUpper
    colorAndOpacity {colorUseRule=UseColor_Specified, specifiedColor=<rgba 0.9301 1 0.0091 1>}
    font {
     bForceMonospaced=false
     bMaterialIsStencil=false
     fontMaterial=null
     fontObject=&BurbankSmall
     letterSpacing=0
     monospacedWidth=1
     size=58
     skewAmount=0
     typefaceFontName=Black
     outlineSettings {
      bApplyOutlineToDropShadows=false
      bMiteredCorners=false
      bSeparateFillAlpha=false
      outlineMaterial=null
      outlineSize=5
      outlineColor=<rgba 0 0 0 1>
     }
    }
    shadowOffset=<0 0>
   }
   slot {class=OverlaySlot, properties {padding=<ltrb -10 -53 0 0>}}
  }
  CommonBorder_0 {
   widgetClassPath=&CommonBorder
   parent=Overlay_1
   widget=CommonBorder_0
   slot {class=OverlaySlot, properties {horizontalAlignment=HAlign_Fill, verticalAlignment=VAlign_Fill, padding=<ltrb 50 24 104 40>}}
  }
  SizeBox_0 {
   widgetClassPath=&SizeBox
   parent=CommonBorder_0
   widget=SizeBox_0
   properties {bOverride_MaxDesiredWidth=true, maxDesiredWidth=760, minDesiredWidth=800}
   slot {class=BorderSlot}
  }
  TextBlock_Description {
   widgetClassPath=&CommonTextBlock
   bInherited=true
   parent=SizeBox_0
   widget=TextBlock_Description
   properties {
    autoWrapText=true
    lineHeightPercentage=0.9
    minDesiredWidth=680
    text="Create and edit your own HUD Layouts!"
    font {
     bForceMonospaced=false
     bMaterialIsStencil=false
     fontMaterial=null
     fontObject=&Burbank
     letterSpacing=0
     monospacedWidth=1
     size=48
     skewAmount=0
     typefaceFontName=Burbank-PlayerPower
     outlineSettings {
      bApplyOutlineToDropShadows=false
      bMiteredCorners=false
      bSeparateFillAlpha=false
      outlineMaterial=null
      outlineSize=4
      outlineColor=<rgba 0 0 0 1>
     }
    }
    shadowOffset=<0 0>
   }
   slot {class=SizeBoxSlot}
  }
 ]
}
