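# Widget Blueprint for the tutorial popup of the HUD Layout Tool V2, introducing layouts, modes, properties and the catalog.
# An inherited visibility switcher holds one overlay per topic, over a transparent blocker border, with the empty event graph leaving logic to the C++ parent class.
# Overlays use hand-positioned outlines, arrows and text prompts, plus mock screen previews built from textless images with live text on top.
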WidgetBlueprint /FortniteCoreUI/Athena/HUD/Bacchus/HUDLayoutTool/HUDLayoutToolV2_TutorialPopup
graphs [EventGraph {links [], nodes []}]
variables []
widgets {
 info {inheritedWidgetCount=2, parentClass=/Script/FortniteUI.HUDLayoutToolV2_TutorialPopup, rootWidgetClass=&ScaleBox, widgetCount=62}
 widgets [
  ScaleBox_0 {widgetClassPath=&ScaleBox, widget=ScaleBox_0, properties {stretch=Fill}}
  Overlay_197 {
   widgetClassPath=&Overlay
   parent=ScaleBox_0
   widget=Overlay_197
   slot {class=ScaleBoxSlot, properties {horizontalAlignment=HAlign_Fill, verticalAlignment=VAlign_Fill}}
  }
  Border_Blocker {
   widgetClassPath=&Border
   parent=Overlay_197
   widget=Border_Blocker
   properties {brushColor=<rgba 1 1 1 0>, padding=<ltrb 0 0 0 0>}
   slot {class=OverlaySlot, properties {horizontalAlignment=HAlign_Fill, verticalAlignment=VAlign_Fill}}
  }
  Switcher_TutorialOverlays {
   widgetClassPath=&CommonVisibilitySwitcher
   bInherited=true
   parent=Overlay_197
   widget=Switcher_TutorialOverlays
   slot {class=OverlaySlot, properties {horizontalAlignment=HAlign_Fill, verticalAlignment=VAlign_Fill}}
  }
  Overlay1_Profiles {
   widgetClassPath=&Overlay
   parent=Switcher_TutorialOverlays
   widget=Overlay1_Profiles
   properties {visibility=Visible}
   slot {class=CommonVisibilitySwitcherSlot}
  }
  ProfileButton_Outline {
   widgetClassPath=&Border
   parent=Overlay1_Profiles
   widget=ProfileButton_Outline
   properties {
    background {
     bIsDynamicallyLoaded=false
     drawAs=Box
     imageType=NoImage
     mirroring=NoMirror
     resourceName=None
     resourceObject=/Game/Athena/HUD/Bacchus/HUDLayoutTool/HLT_Icons/M_HUDLayoutToolV2_Tutorial_PreviewOutline
     tiling=NoTile
     imageSize=<32 32>
     margin=<ltrb 0.5 0.5 0.5 0.5>
     outlineSettings {
      bUseBrushTransparency=false
      roundingType=HalfHeightRadius
      width=0
      color {colorUseRule=UseColor_Specified, specifiedColor=<rgba 0 0 0 0>}
      cornerRadii=<0 0 0 0>
     }
     tintColor {colorUseRule=UseColor_Specified, specifiedColor=<rgba 1 1 1 1>}
     uVRegion {bIsValid=false, max=<0 0>, min=<0 0>}
    }
    brushColor=<rgba 0.9301 1 0.0091 1>
    padding=<ltrb 256 52 256 52>
   }
   slot {class=OverlaySlot, properties {horizontalAlignment=HAlign_Center, padding=<ltrb 0 38 195 0>}}
  }
  Overlay_ClippingMask {
   widgetClassPath=&Overlay
   parent=Overlay1_Profiles
   widget=Overlay_ClippingMask
   slot {class=OverlaySlot, properties {horizontalAlignment=HAlign_Center, padding=<ltrb 0 -20 600 0>}}
  }
  Image_Arrow {
   widgetClassPath=&Image
   bIsVariable=true
   parent=Overlay_ClippingMask
   widget=Image_Arrow
   properties {
    clipping=ClipToBounds
    brush {
     bIsDynamicallyLoaded=false
     drawAs=Image
     imageType=NoImage
     mirroring=NoMirror
     resourceName=None
     resourceObject=/Game/Athena/HUD/Bacchus/HUDLayoutTool/HLT_Icons/T-HLT-Tutorial_Arrow_Sprite
     tiling=NoTile
     imageSize=<392 800>
     margin=<ltrb 0 0 0 0>
     outlineSettings {
      bUseBrushTransparency=false
      roundingType=HalfHeightRadius
      width=0
      color {colorUseRule=UseColor_Specified, specifiedColor=<rgba 0 0 0 0>}
      cornerRadii=<0 0 0 0>
     }
     tintColor {colorUseRule=UseColor_Specified, specifiedColor=<rgba 1 1 1 1>}
     uVRegion {bIsValid=false, max=<0 0>, min=<0 0>}
    }
   }
   slot {class=OverlaySlot}
  }
  Overlay_ProfilesPreview {
   widgetClassPath=&Overlay
   parent=Overlay1_Profiles
   widget=Overlay_ProfilesPreview
   properties {renderTransform {angle=3, scale=<1 1>, shear=<0 0>, translation=<0 0>}}
   slot {class=OverlaySlot, properties {horizontalAlignment=HAlign_Center, padding=<ltrb 0 500 330 0>}}
  }
  Image_BGDropShadow {
   widgetClassPath=&Image
   parent=Overlay_ProfilesPreview
   widget=Image_BGDropShadow
   properties {
    visibility=SelfHitTestInvisible
    brush {
     bIsDynamicallyLoaded=false
     drawAs=Border
     imageType=NoImage
     mirroring=NoMirror
     resourceName=None
     resourceObject=/Game/Athena/HUD/Bacchus/HUDLayoutTool/HLT_Icons/T-HLT-TutorialShadow
     tiling=NoTile
     imageSize=<64 64>
     margin=<ltrb 0.5 0.5 0.5 0.5>
     outlineSettings {
      bUseBrushTransparency=false
      roundingType=HalfHeightRadius
      width=0
      color {colorUseRule=UseColor_Specified, specifiedColor=<rgba 0 0 0 0>}
      cornerRadii=<0 0 0 0>
     }
     tintColor {colorUseRule=UseColor_Specified, specifiedColor=<rgba 1 1 1 1>}
     uVRegion {bIsValid=false, max=<0 0>, min=<0 0>}
    }
    colorAndOpacity=<rgba 0 0 0 0.8>
   }
   slot {class=OverlaySlot, properties {horizontalAlignment=HAlign_Fill, verticalAlignment=VAlign_Fill, padding=<ltrb -30 -30 -30 -30>}}
  }
  CanvasPanel_ProfilePreview {
   widgetClassPath=&CanvasPanel
   parent=Overlay_ProfilesPreview
   widget=CanvasPanel_ProfilePreview
   slot {class=OverlaySlot, properties {horizontalAlignment=HAlign_Fill, verticalAlignment=VAlign_Fill}}
  }
  Image_Preview {
   widgetClassPath=&Image
   bIsVariable=true
   parent=CanvasPanel_ProfilePreview
   widget=Image_Preview
   properties {brush {
     bIsDynamicallyLoaded=false
     drawAs=Image
     imageType=NoImage
     mirroring=NoMirror
     resourceName=None
     resourceObject=/Game/Athena/HUD/Bacchus/HUDLayoutTool/HLT_Icons/T-HLT-Tutorial-NoText_ProfilePreview
     tiling=NoTile
     imageSize=<866 486>
     margin=<ltrb 0 0 0 0>
     outlineSettings {
      bUseBrushTransparency=false
      roundingType=HalfHeightRadius
      width=0
      color {colorUseRule=UseColor_Specified, specifiedColor=<rgba 0 0 0 0>}
      cornerRadii=<0 0 0 0>
     }
     tintColor {colorUseRule=UseColor_Specified, specifiedColor=<rgba 1 1 1 1>}
     uVRegion {bIsValid=false, max=<0 0>, min=<0 0>}
    }}
   slot {
    class=CanvasPanelSlot
    properties {bAutoSize=true, layoutData {alignment=<0 0>, anchors {maximum=<0 0>, minimum=<0 0>}, offsets=<ltrb 0 0 866 486>}}
   }
  }
  ScreenTitle {
   widgetClassPath=&CommonTextBlock
   parent=CanvasPanel_ProfilePreview
   widget=ScreenTitle
   properties {
    scrollStyle=&TextScrollStyle-Base_C
    text=HUD Layouts
    textTransformPolicy=ToUpper
    font {
     bForceMonospaced=false
     bMaterialIsStencil=false
     fontMaterial=null
     fontObject=&BurbankSmall
     letterSpacing=0
     monospacedWidth=1
     size=22
     skewAmount=0
     typefaceFontName=Black
     outlineSettings {
      bApplyOutlineToDropShadows=false
      bMiteredCorners=false
      bSeparateFillAlpha=false
      outlineMaterial=null
      outlineSize=0
      outlineColor=<rgba 0 0 0 1>
     }
    }
    renderTransform {angle=0, scale=<1 1>, shear=<-10 0>, translation=<0 0>}
    shadowOffset=<0 0>
   }
   slot {
    class=CanvasPanelSlot
    properties {layoutData {alignment=<0 0>, anchors {maximum=<0 0>, minimum=<0 0>}, offsets=<ltrb 47.7354 18.2558 692.0564 30>}}
   }
  }
  PresetSubTitle {
   widgetClassPath=&CommonTextBlock
   parent=CanvasPanel_ProfilePreview
   widget=PresetSubTitle
   properties {
    scrollStyle=&TextScrollStyle-Base_C
    text=Presets
    textTransformPolicy=ToUpper
    colorAndOpacity {colorUseRule=UseColor_Specified, specifiedColor=<rgba 0.1529 0.3515 0.807 1>}
    font {
     bForceMonospaced=false
     bMaterialIsStencil=false
     fontMaterial=null
     fontObject=&BurbankSmall
     letterSpacing=0
     monospacedWidth=1
     size=14
     skewAmount=0
     typefaceFontName=Black
     outlineSettings {
      bApplyOutlineToDropShadows=false
      bMiteredCorners=false
      bSeparateFillAlpha=false
      outlineMaterial=null
      outlineSize=0
      outlineColor=<rgba 0 0 0 1>
     }
    }
    shadowOffset=<0 0>
   }
   slot {
    class=CanvasPanelSlot
    properties {layoutData {alignment=<0 0.5>, anchors {maximum=<0 0.5>, minimum=<0 0.5>}, offsets=<ltrb 45.3463 -129 169.8198 30>}}
   }
  }
  BuilderProPreset {
   widgetClassPath=&CommonTextBlock
   parent=CanvasPanel_ProfilePreview
   widget=BuilderProPreset
   properties {
    justification=Center
    scrollStyle=&TextScrollStyle-Base_C
    text=Builder Pro
    textTransformPolicy=ToUpper
    font {
     bForceMonospaced=false
     bMaterialIsStencil=false
     fontMaterial=null
     fontObject=&BurbankSmall
     letterSpacing=0
     monospacedWidth=1
     size=17
     skewAmount=0
     typefaceFontName=Black
     outlineSettings {
      bApplyOutlineToDropShadows=false
      bMiteredCorners=false
      bSeparateFillAlpha=false
      outlineMaterial=null
      outlineSize=0
      outlineColor=<rgba 0 0 0 1>
     }
    }
    shadowOffset=<0 0>
   }
   slot {
    class=CanvasPanelSlot
    properties {layoutData {alignment=<0 0.5>, anchors {maximum=<0 0.5>, minimum=<0 0.5>}, offsets=<ltrb 72.4013 -84.3157 190.8966 30>}}
   }
  }
  CombatProPreset {
   widgetClassPath=&CommonTextBlock
   parent=CanvasPanel_ProfilePreview
   widget=CombatProPreset
   properties {
    justification=Center
    scrollStyle=&TextScrollStyle-Base_C
    text=Combat Pro
    textTransformPolicy=ToUpper
    font {
     bForceMonospaced=false
     bMaterialIsStencil=false
     fontMaterial=null
     fontObject=&BurbankSmall
     letterSpacing=0
     monospacedWidth=1
     size=17
     skewAmount=0
     typefaceFontName=Black
     outlineSettings {
      bApplyOutlineToDropShadows=false
      bMiteredCorners=false
      bSeparateFillAlpha=false
      outlineMaterial=null
      outlineSize=0
      outlineColor=<rgba 0 0 0 1>
     }
    }
    shadowOffset=<0 0>
   }
   slot {
    class=CanvasPanelSlot
    properties {layoutData {alignment=<0 0.5>, anchors {maximum=<0 0.5>, minimum=<0 0.5>}, offsets=<ltrb 329.5 -84.3157 190.8966 30>}}
   }
  }
  OldSchoolPreset {
   widgetClassPath=&CommonTextBlock
   parent=CanvasPanel_ProfilePreview
   widget=OldSchoolPreset
   properties {
    justification=Center
    scrollStyle=&TextScrollStyle-Base_C
    text=Old School
    textTransformPolicy=ToUpper
    font {
     bForceMonospaced=false
     bMaterialIsStencil=false
     fontMaterial=null
     fontObject=&BurbankSmall
     letterSpacing=0
     monospacedWidth=1
     size=17
     skewAmount=0
     typefaceFontName=Black
     outlineSettings {
      bApplyOutlineToDropShadows=false
      bMiteredCorners=false
      bSeparateFillAlpha=false
      outlineMaterial=null
      outlineSize=0
      outlineColor=<rgba 0 0 0 1>
     }
    }
    shadowOffset=<0 0>
   }
   slot {
    class=CanvasPanelSlot
    properties {layoutData {alignment=<0 0.5>, anchors {maximum=<0 0.5>, minimum=<0 0.5>}, offsets=<ltrb 586 -84.3157 190.8966 30>}}
   }
  }
  CustomSubTitle {
   widgetClassPath=&CommonTextBlock
   parent=CanvasPanel_ProfilePreview
   widget=CustomSubTitle
   properties {
    scrollStyle=&TextScrollStyle-Base_C
    text=Custom Layouts
    textTransformPolicy=ToUpper
    colorAndOpacity {colorUseRule=UseColor_Specified, specifiedColor=<rgba 0.1529 0.3515 0.807 1>}
    font {
     bForceMonospaced=false
     bMaterialIsStencil=false
     fontMaterial=null
     fontObject=&BurbankSmall
     letterSpacing=0
     monospacedWidth=1
     size=14
     skewAmount=0
     typefaceFontName=Black
     outlineSettings {
      bApplyOutlineToDropShadows=false
      bMiteredCorners=false
      bSeparateFillAlpha=false
      outlineMaterial=null
      outlineSize=0
      outlineColor=<rgba 0 0 0 1>
     }
    }
    shadowOffset=<0 0>
   }
   slot {
    class=CanvasPanelSlot
    properties {layoutData {alignment=<0 0.5>, anchors {maximum=<0 0.5>, minimum=<0 0.5>}, offsets=<ltrb 45.3463 -12 242.2853 30>}}
   }
  }
  CustomLayout {
   widgetClassPath=&CommonTextBlock
   parent=CanvasPanel_ProfilePreview
   widget=CustomLayout
   properties {
    justification=Center
    scrollStyle=&TextScrollStyle-Base_C
    text=Custom Layout
    textTransformPolicy=ToUpper
    colorAndOpacity {colorUseRule=UseColor_Specified, specifiedColor=<rgba 0 0 0 1>}
    font {
     bForceMonospaced=false
     bMaterialIsStencil=false
     fontMaterial=null
     fontObject=&BurbankSmall
     letterSpacing=0
     monospacedWidth=1
     size=17
     skewAmount=0
     typefaceFontName=Black
     outlineSettings {
      bApplyOutlineToDropShadows=false
      bMiteredCorners=false
      bSeparateFillAlpha=false
      outlineMaterial=null
      outlineSize=0
      outlineColor=<rgba 0 0 0 1>
     }
    }
    shadowOffset=<0 0>
   }
   slot {
    class=CanvasPanelSlot
    properties {layoutData {alignment=<0 0.5>, anchors {maximum=<0 0.5>, minimum=<0 0.5>}, offsets=<ltrb 72.4013 30.7873 190.8966 30>}}
   }
  }
  Text_ActiveIndicator {
   widgetClassPath=&CommonTextBlock
   parent=CanvasPanel_ProfilePreview
   widget=Text_ActiveIndicator
   properties {
    clipping=OnDemand
    justification=Center
    scrollStyle=&TextScrollStyle-Base_C
    text=Active
    textTransformPolicy=ToUpper
    colorAndOpacity {colorUseRule=UseColor_Specified, specifiedColor=<rgba 0 0 0 1>}
    font {
     bForceMonospaced=false
     bMaterialIsStencil=false
     fontMaterial=null
     fontObject=&BurbankSmall
     letterSpacing=0
     monospacedWidth=1
     size=10
     skewAmount=0
     typefaceFontName=Black
     outlineSettings {
      bApplyOutlineToDropShadows=false
      bMiteredCorners=false
      bSeparateFillAlpha=false
      outlineMaterial=null
      outlineSize=0
      outlineColor=<rgba 0 0 0 1>
     }
    }
    shadowOffset=<0 0>
   }
   slot {
    class=CanvasPanelSlot
    properties {layoutData {alignment=<0 0.5>, anchors {maximum=<0 0.5>, minimum=<0 0.5>}, offsets=<ltrb 233.6893 -115 54 15>}}
   }
  }
  DeleteButton {
   widgetClassPath=&CommonTextBlock
   parent=CanvasPanel_ProfilePreview
   widget=DeleteButton
   properties {
    justification=Center
    scrollStyle=&TextScrollStyle-Base_C
    text=Delete
    textTransformPolicy=ToUpper
    colorAndOpacity {colorUseRule=UseColor_Specified, specifiedColor=<rgba 1 0.1714 0.1329 1>}
    font {
     bForceMonospaced=false
     bMaterialIsStencil=false
     fontMaterial=null
     fontObject=&BurbankSmall
     letterSpacing=0
     monospacedWidth=1
     size=12
     skewAmount=0
     typefaceFontName=Black
     outlineSettings {
      bApplyOutlineToDropShadows=false
      bMiteredCorners=false
      bSeparateFillAlpha=false
      outlineMaterial=null
      outlineSize=0
      outlineColor=<rgba 0 0 0 1>
     }
    }
    shadowOffset=<0 0>
   }
   slot {
    class=CanvasPanelSlot
    properties {layoutData {alignment=<0 1>, anchors {maximum=<0 1>, minimum=<0 1>}, offsets=<ltrb 55 -30.3 108 20.6095>}}
   }
  }
  RenameButton {
   widgetClassPath=&CommonTextBlock
   parent=CanvasPanel_ProfilePreview
   widget=RenameButton
   properties {
    justification=Center
    scrollStyle=&TextScrollStyle-Base_C
    text=Rename
    textTransformPolicy=ToUpper
    colorAndOpacity {colorUseRule=UseColor_Specified, specifiedColor=<rgba 0.0319 0.7835 0.9473 1>}
    font {
     bForceMonospaced=false
     bMaterialIsStencil=false
     fontMaterial=null
     fontObject=&BurbankSmall
     letterSpacing=0
     monospacedWidth=1
     size=12
     skewAmount=0
     typefaceFontName=Black
     outlineSettings {
      bApplyOutlineToDropShadows=false
      bMiteredCorners=false
      bSeparateFillAlpha=false
      outlineMaterial=null
      outlineSize=0
      outlineColor=<rgba 0 0 0 1>
     }
    }
    shadowOffset=<0 0>
   }
   slot {
    class=CanvasPanelSlot
    properties {layoutData {alignment=<0 1>, anchors {maximum=<0 1>, minimum=<0 1>}, offsets=<ltrb 195.1871 -30.3 108 20.6095>}}
   }
  }
  CreateNewButton {
   widgetClassPath=&CommonTextBlock
   parent=CanvasPanel_ProfilePreview
   widget=CreateNewButton
   properties {
    justification=Center
    scrollStyle=&TextScrollStyle-Base_C
    text=Create New
    textTransformPolicy=ToUpper
    colorAndOpacity {colorUseRule=UseColor_Specified, specifiedColor=<rgba 0.0319 0.7835 0.9473 1>}
    font {
     bForceMonospaced=false
     bMaterialIsStencil=false
     fontMaterial=null
     fontObject=&BurbankSmall
     letterSpacing=0
     monospacedWidth=1
     size=12
     skewAmount=0
     typefaceFontName=Black
     outlineSettings {
      bApplyOutlineToDropShadows=false
      bMiteredCorners=false
      bSeparateFillAlpha=false
      outlineMaterial=null
      outlineSize=0
      outlineColor=<rgba 0 0 0 1>
     }
    }
    shadowOffset=<0 0>
   }
   slot {
    class=CanvasPanelSlot
    properties {layoutData {alignment=<0 1>, anchors {maximum=<0 1>, minimum=<0 1>}, offsets=<ltrb 333 -30.3 108 20.6095>}}
   }
  }
  SelectLayoutButton {
   widgetClassPath=&CommonTextBlock
   parent=CanvasPanel_ProfilePreview
   widget=SelectLayoutButton
   properties {
    justification=Center
    scrollStyle=&TextScrollStyle-Base_C
    text=Select Layout
    textTransformPolicy=ToUpper
    colorAndOpacity {colorUseRule=UseColor_Specified, specifiedColor=<rgba 0 0 0 1>}
    font {
     bForceMonospaced=false
     bMaterialIsStencil=false
     fontMaterial=null
     fontObject=&BurbankSmall
     letterSpacing=0
     monospacedWidth=1
     size=12
     skewAmount=0
     typefaceFontName=Black
     outlineSettings {
      bApplyOutlineToDropShadows=false
      bMiteredCorners=false
      bSeparateFillAlpha=false
      outlineMaterial=null
      outlineSize=0
      outlineColor=<rgba 0 0 0 1>
     }
    }
    shadowOffset=<0 0>
   }
   slot {
    class=CanvasPanelSlot
    properties {layoutData {alignment=<1 1>, anchors {maximum=<1 1>, minimum=<1 1>}, offsets=<ltrb -56 -30.3 126 20.6095>}}
   }
  }
  PreviewOutline {
   widgetClassPath=&Border
   parent=Overlay_ProfilesPreview
   widget=PreviewOutline
   properties {
    background {
     bIsDynamicallyLoaded=false
     drawAs=Box
     imageType=NoImage
     mirroring=NoMirror
     resourceName=None
     resourceObject=&M_HUDLayoutToolV2_Tutorial_PreviewOutline
     tiling=NoTile
     imageSize=<32 32>
     margin=<ltrb 0.5 0.5 0.5 0.5>
     outlineSettings {
      bUseBrushTransparency=false
      roundingType=HalfHeightRadius
      width=0
      color {colorUseRule=UseColor_Specified, specifiedColor=<rgba 0 0 0 0>}
      cornerRadii=<0 0 0 0>
     }
     tintColor {colorUseRule=UseColor_Specified, specifiedColor=<rgba 1 1 1 1>}
     uVRegion {bIsValid=false, max=<0 0>, min=<0 0>}
    }
    brushColor=<rgba 0.9301 1 0.0091 1>
   }
   slot {class=OverlaySlot, properties {horizontalAlignment=HAlign_Fill, verticalAlignment=VAlign_Fill, padding=<ltrb -5 -7 -7 -8>}}
  }
  Border_NewTag {
   widgetClassPath=&CommonBorder
   bIsVariable=true
   parent=Overlay_ProfilesPreview
   widget=Border_NewTag
   properties {
    bFlipForRightToLeftFlowDirection=true
    visibility=Collapsed
    background {
     bIsDynamicallyLoaded=false
     drawAs=Box
     imageType=NoImage
     mirroring=NoMirror
     resourceName=None
     resourceObject=/Game/Athena/HUD/Bacchus/HUDLayoutTool/HLT_Icons/T-HLT-Tutorial_NewTag_Sprite
     tiling=NoTile
     imageSize=<128 64>
     margin=<ltrb 0.1 0 0.25 0>
     outlineSettings {
      bUseBrushTransparency=false
      roundingType=HalfHeightRadius
      width=0
      color {colorUseRule=UseColor_Specified, specifiedColor=<rgba 0 0 0 0>}
      cornerRadii=<0 0 0 0>
     }
     tintColor {colorUseRule=UseColor_Specified, specifiedColor=<rgba 1 1 1 1>}
     uVRegion {bIsValid=false, max=<0 0>, min=<0 0>}
    }
    padding=<ltrb 26 14 40 10>
   }
   slot {class=OverlaySlot, properties {horizontalAlignment=HAlign_Right, padding=<ltrb 0 -50 -30 0>}}
  }
  SizeBox_0 {
   widgetClassPath=&SizeBox
   parent=Border_NewTag
   widget=SizeBox_0
   properties {bOverride_MaxDesiredWidth=true, maxDesiredWidth=256}
   slot {class=BorderSlot, properties {padding=<ltrb 26 14 40 10>}}
  }
  Text_Tag {
   widgetClassPath=&CommonTextBlock
   bIsVariable=true
   parent=SizeBox_0
   widget=Text_Tag
   properties {
    text="New!"
    textTransformPolicy=ToUpper
    font {
     bForceMonospaced=false
     bMaterialIsStencil=false
     fontMaterial=null
     fontObject=&Burbank
     letterSpacing=0
     monospacedWidth=1
     size=36
     skewAmount=0
     typefaceFontName=Burbank-Black
     outlineSettings {
      bApplyOutlineToDropShadows=false
      bMiteredCorners=false
      bSeparateFillAlpha=false
      outlineMaterial=null
      outlineSize=0
      outlineColor=<rgba 0 0 0 1>
     }
    }
    renderTransform {angle=0, scale=<1 0.86>, shear=<0 0>, translation=<0 0>}
    shadowOffset=<0 0>
   }
   slot {class=SizeBoxSlot}
  }
  HUDLayoutToolV2_TutorialTextPrompt {
   widgetClassPath=&/HUDLayoutToolV2_TutorialTextPrompt
   bIsVariable=true
   parent=Overlay1_Profiles
   widget=HUDLayoutToolV2_TutorialTextPrompt
   properties {
    descriptionText="Create and edit your own HUD Layouts!"
    titleText="Hud Layouts!"
    renderTransform {angle=-3, scale=<1 1>, shear=<0 0>, translation=<0 0>}
   }
   slot {class=OverlaySlot, properties {horizontalAlignment=HAlign_Center, padding=<ltrb 0 290 440 0>}}
  }
  Overlay2_Contexts {
   widgetClassPath=&Overlay
   parent=Switcher_TutorialOverlays
   widget=Overlay2_Contexts
   properties {visibility=Visible}
   slot {class=CommonVisibilitySwitcherSlot}
  }
  ContextButton_Outline {
   widgetClassPath=&Border
   parent=Overlay2_Contexts
   widget=ContextButton_Outline
   properties {
    background {
     bIsDynamicallyLoaded=false
     drawAs=Box
     imageType=NoImage
     mirroring=NoMirror
     resourceName=None
     resourceObject=&M_HUDLayoutToolV2_Tutorial_PreviewOutline
     tiling=NoTile
     imageSize=<32 32>
     margin=<ltrb 0.5 0.5 0.5 0.5>
     outlineSettings {
      bUseBrushTransparency=false
      roundingType=HalfHeightRadius
      width=0
      color {colorUseRule=UseColor_Specified, specifiedColor=<rgba 0 0 0 0>}
      cornerRadii=<0 0 0 0>
     }
     tintColor {colorUseRule=UseColor_Specified, specifiedColor=<rgba 1 1 1 1>}
     uVRegion {bIsValid=false, max=<0 0>, min=<0 0>}
    }
    brushColor=<rgba 0.9301 1 0.0091 1>
    padding=<ltrb 128 52 128 50>
   }
   slot {class=OverlaySlot, properties {horizontalAlignment=HAlign_Center, padding=<ltrb 0 140 324 0>}}
  }
  Image_Arrow_1 {
   widgetClassPath=&Image
   bIsVariable=true
   parent=Overlay2_Contexts
   widget=Image_Arrow_1
   properties {brush {
     bIsDynamicallyLoaded=false
     drawAs=Image
     imageType=NoImage
     mirroring=NoMirror
     resourceName=None
     resourceObject=&T-HLT-Tutorial_Arrow_Sprite
     tiling=NoTile
     imageSize=<392 800>
     margin=<ltrb 0 0 0 0>
     outlineSettings {
      bUseBrushTransparency=false
      roundingType=HalfHeightRadius
      width=0
      color {colorUseRule=UseColor_Specified, specifiedColor=<rgba 0 0 0 0>}
      cornerRadii=<0 0 0 0>
     }
     tintColor {colorUseRule=UseColor_Specified, specifiedColor=<rgba 1 1 1 1>}
     uVRegion {bIsValid=false, max=<0 0>, min=<0 0>}
    }}
   slot {class=OverlaySlot, properties {horizontalAlignment=HAlign_Center, padding=<ltrb 0 70 600 0>}}
  }
  Overlay_ContextPreview {
   widgetClassPath=&Overlay
   parent=Overlay2_Contexts
   widget=Overlay_ContextPreview
   slot {class=OverlaySlot, properties {horizontalAlignment=HAlign_Center, padding=<ltrb 0 245 0 0>}}
  }
  PreviewOutline_1 {
   widgetClassPath=&Border
   parent=Overlay_ContextPreview
   widget=PreviewOutline_1
   properties {
    background {
     bIsDynamicallyLoaded=false
     drawAs=Box
     imageType=NoImage
     mirroring=NoMirror
     resourceName=None
     resourceObject=&M_HUDLayoutToolV2_Tutorial_PreviewOutline
     tiling=NoTile
     imageSize=<32 32>
     margin=<ltrb 0.4 0.36 0.4 0.5>
     outlineSettings {
      bUseBrushTransparency=false
      roundingType=HalfHeightRadius
      width=0
      color {colorUseRule=UseColor_Specified, specifiedColor=<rgba 0 0 0 0>}
      cornerRadii=<0 0 0 0>
     }
     tintColor {colorUseRule=UseColor_Specified, specifiedColor=<rgba 1 1 1 1>}
     uVRegion {bIsValid=false, max=<0 0>, min=<0 0>}
    }
    brushColor=<rgba 0.9301 1 0.0091 1>
    padding=<ltrb 472 210 472 210>
   }
   slot {class=OverlaySlot, properties {horizontalAlignment=HAlign_Fill, verticalAlignment=VAlign_Fill}}
  }
  Border_NewTag_1 {
   widgetClassPath=&CommonBorder
   bIsVariable=true
   parent=Overlay_ContextPreview
   widget=Border_NewTag_1
   properties {
    bFlipForRightToLeftFlowDirection=true
    visibility=Collapsed
    background {
     bIsDynamicallyLoaded=false
     drawAs=Box
     imageType=NoImage
     mirroring=NoMirror
     resourceName=None
     resourceObject=&T-HLT-Tutorial_NewTag_Sprite
     tiling=NoTile
     imageSize=<128 64>
     margin=<ltrb 0.1 0 0.25 0>
     outlineSettings {
      bUseBrushTransparency=false
      roundingType=HalfHeightRadius
      width=0
      color {colorUseRule=UseColor_Specified, specifiedColor=<rgba 0 0 0 0>}
      cornerRadii=<0 0 0 0>
     }
     tintColor {colorUseRule=UseColor_Specified, specifiedColor=<rgba 1 1 1 1>}
     uVRegion {bIsValid=false, max=<0 0>, min=<0 0>}
    }
    padding=<ltrb 26 14 40 10>
   }
   slot {class=OverlaySlot, properties {horizontalAlignment=HAlign_Right, padding=<ltrb 0 -45 -20 0>}}
  }
  SizeBox {
   widgetClassPath=&SizeBox
   parent=Border_NewTag_1
   widget=SizeBox
   properties {bOverride_MaxDesiredWidth=true, maxDesiredWidth=256}
   slot {class=BorderSlot, properties {padding=<ltrb 26 14 40 10>}}
  }
  Text_Tag_1 {
   widgetClassPath=&CommonTextBlock
   bIsVariable=true
   parent=SizeBox
   widget=Text_Tag_1
   properties {
    text="New!"
    textTransformPolicy=ToUpper
    font {
     bForceMonospaced=false
     bMaterialIsStencil=false
     fontMaterial=null
     fontObject=&Burbank
     letterSpacing=0
     monospacedWidth=1
     size=36
     skewAmount=0
     typefaceFontName=Burbank-Black
     outlineSettings {
      bApplyOutlineToDropShadows=false
      bMiteredCorners=false
      bSeparateFillAlpha=false
      outlineMaterial=null
      outlineSize=0
      outlineColor=<rgba 0 0 0 1>
     }
    }
    renderTransform {angle=0, scale=<1 0.86>, shear=<0 0>, translation=<0 0>}
    shadowOffset=<0 0>
   }
   slot {class=SizeBoxSlot}
  }
  HUDLayoutToolV2_TutorialTextPrompt_1 {
   widgetClassPath=&HUDLayoutToolV2_TutorialTextPrompt
   bIsVariable=true
   parent=Overlay2_Contexts
   widget=HUDLayoutToolV2_TutorialTextPrompt_1
   properties {
    descriptionText="Edit controls for combat mode, build mode, vehicle modes, and more!"
    titleText="Modes!"
    renderTransform {angle=-3, scale=<1 1>, shear=<0 0>, translation=<0 0>}
   }
   slot {class=OverlaySlot, properties {horizontalAlignment=HAlign_Center, padding=<ltrb 70 740 520 0>}}
  }
  Overlay3_Properties {
   widgetClassPath=&Overlay
   parent=Switcher_TutorialOverlays
   widget=Overlay3_Properties
   properties {visibility=Visible}
   slot {class=CommonVisibilitySwitcherSlot}
  }
  PropertiesButton_Outline {
   widgetClassPath=&Border
   parent=Overlay3_Properties
   widget=PropertiesButton_Outline
   properties {
    background {
     bIsDynamicallyLoaded=false
     drawAs=Box
     imageType=NoImage
     mirroring=NoMirror
     resourceName=None
     resourceObject=&M_HUDLayoutToolV2_Tutorial_PreviewOutline
     tiling=NoTile
     imageSize=<32 32>
     margin=<ltrb 0.5 0.5 0.5 0.5>
     outlineSettings {
      bUseBrushTransparency=false
      roundingType=HalfHeightRadius
      width=0
      color {colorUseRule=UseColor_Specified, specifiedColor=<rgba 0 0 0 0>}
      cornerRadii=<0 0 0 0>
     }
     tintColor {colorUseRule=UseColor_Specified, specifiedColor=<rgba 1 1 1 1>}
     uVRegion {bIsValid=false, max=<0 0>, min=<0 0>}
    }
    brushColor=<rgba 0.9301 1 0.0091 1>
    padding=<ltrb 128 52 128 50>
   }
   slot {class=OverlaySlot, properties {horizontalAlignment=HAlign_Center, padding=<ltrb 0 140 68 0>}}
  }
  Image_Arrow_2 {
   widgetClassPath=&Image
   bIsVariable=true
   parent=Overlay3_Properties
   widget=Image_Arrow_2
   properties {brush {
     bIsDynamicallyLoaded=false
     drawAs=Image
     imageType=NoImage
     mirroring=NoMirror
     resourceName=None
     resourceObject=&T-HLT-Tutorial_Arrow_Sprite
     tiling=NoTile
     imageSize=<-392 800>
     margin=<ltrb 0 0 0 0>
     outlineSettings {
      bUseBrushTransparency=false
      roundingType=HalfHeightRadius
      width=0
      color {colorUseRule=UseColor_Specified, specifiedColor=<rgba 0 0 0 0>}
      cornerRadii=<0 0 0 0>
     }
     tintColor {colorUseRule=UseColor_Specified, specifiedColor=<rgba 1 1 1 1>}
     uVRegion {bIsValid=false, max=<0 0>, min=<0 0>}
    }}
   slot {class=OverlaySlot, properties {horizontalAlignment=HAlign_Center, padding=<ltrb 206 70 0 0>}}
  }
  Overlay_PropertiesPreview {
   widgetClassPath=&Overlay
   parent=Overlay3_Properties
   widget=Overlay_PropertiesPreview
   properties {renderTransform {angle=-3, scale=<1 1>, shear=<0 0>, translation=<0 0>}}
   slot {class=OverlaySlot, properties {horizontalAlignment=HAlign_Center, padding=<ltrb 358 560 0 0>}}
  }
  Image_BGDropShadow_1 {
   widgetClassPath=&Image
   parent=Overlay_PropertiesPreview
   widget=Image_BGDropShadow_1
   properties {
    visibility=SelfHitTestInvisible
    brush {
     bIsDynamicallyLoaded=false
     drawAs=Border
     imageType=NoImage
     mirroring=NoMirror
     resourceName=None
     resourceObject=&T-HLT-TutorialShadow
     tiling=NoTile
     imageSize=<64 64>
     margin=<ltrb 0.5 0.5 0.5 0.5>
     outlineSettings {
      bUseBrushTransparency=false
      roundingType=HalfHeightRadius
      width=0
      color {colorUseRule=UseColor_Specified, specifiedColor=<rgba 0 0 0 0>}
      cornerRadii=<0 0 0 0>
     }
     tintColor {colorUseRule=UseColor_Specified, specifiedColor=<rgba 1 1 1 1>}
     uVRegion {bIsValid=false, max=<0 0>, min=<0 0>}
    }
    colorAndOpacity=<rgba 0 0 0 0.8>
   }
   slot {class=OverlaySlot, properties {horizontalAlignment=HAlign_Fill, verticalAlignment=VAlign_Fill, padding=<ltrb -30 -30 -30 -30>}}
  }
  CanvasPanel_PropertiesPreview {
   widgetClassPath=&CanvasPanel
   parent=Overlay_PropertiesPreview
   widget=CanvasPanel_PropertiesPreview
   slot {class=OverlaySlot, properties {horizontalAlignment=HAlign_Center, verticalAlignment=VAlign_Center}}
  }
  MainLayoutsTitle {
   widgetClassPath=&CommonTextBlock
   parent=CanvasPanel_PropertiesPreview
   widget=MainLayoutsTitle
   properties {
    scrollStyle=&TextScrollStyle-Base_C
    text=Button
    textTransformPolicy=ToUpper
    font {
     bForceMonospaced=false
     bMaterialIsStencil=false
     fontMaterial=null
     fontObject=&BurbankSmall
     letterSpacing=0
     monospacedWidth=1
     size=28
     skewAmount=0
     typefaceFontName=Black
     outlineSettings {
      bApplyOutlineToDropShadows=false
      bMiteredCorners=false
      bSeparateFillAlpha=false
      outlineMaterial=null
      outlineSize=0
      outlineColor=<rgba 0 0 0 1>
     }
    }
    shadowOffset=<0 0>
   }
   slot {
    class=CanvasPanelSlot
    properties {layoutData {alignment=<0 0>, anchors {maximum=<0 0>, minimum=<0 0>}, offsets=<ltrb 28.8565 27.1836 350 45>}}
   }
  }
  SubTitle1 {
   widgetClassPath=&CommonTextBlock
   parent=CanvasPanel_PropertiesPreview
   widget=SubTitle1
   properties {
    scrollStyle=&TextScrollStyle-Base_C
    text=Button Size
    colorAndOpacity {colorUseRule=UseColor_Specified, specifiedColor=<rgba 0.4125 0.4125 0.4125 1>}
    font {
     bForceMonospaced=false
     bMaterialIsStencil=false
     fontMaterial=null
     fontObject=&BurbankSmall
     letterSpacing=0
     monospacedWidth=1
     size=27
     skewAmount=0
     typefaceFontName=Bold
     outlineSettings {
      bApplyOutlineToDropShadows=false
      bMiteredCorners=false
      bSeparateFillAlpha=false
      outlineMaterial=null
      outlineSize=0
      outlineColor=<rgba 0 0 0 1>
     }
    }
    shadowOffset=<0 0>
   }
   slot {
    class=CanvasPanelSlot
    properties {layoutData {alignment=<0 0>, anchors {maximum=<0 0>, minimum=<0 0>}, offsets=<ltrb 28.8565 85.2964 543.0811 45>}}
   }
  }
  Percent1 {
   widgetClassPath=&CommonTextBlock
   parent=CanvasPanel_PropertiesPreview
   widget=Percent1
   properties {
    justification=Right
    scrollStyle=&TextScrollStyle-Base_C
    text=80%
    textTransformPolicy=ToUpper
    font {
     bForceMonospaced=false
     bMaterialIsStencil=false
     fontMaterial=null
     fontObject=&BurbankSmall
     letterSpacing=0
     monospacedWidth=1
     size=27
     skewAmount=0
     typefaceFontName=Bold
     outlineSettings {
      bApplyOutlineToDropShadows=false
      bMiteredCorners=false
      bSeparateFillAlpha=false
      outlineMaterial=null
      outlineSize=0
      outlineColor=<rgba 0 0 0 1>
     }
    }
    shadowOffset=<0 0>
   }
   slot {
    class=CanvasPanelSlot
    properties {layoutData {alignment=<1 0>, anchors {maximum=<1 0>, minimum=<1 0>}, offsets=<ltrb -126.5203 85.2964 313.2523 45>}}
   }
  }
  SubTitle2 {
   widgetClassPath=&CommonTextBlock
   parent=CanvasPanel_PropertiesPreview
   widget=SubTitle2
   properties {
    scrollStyle=&TextScrollStyle-Base_C
    text=Touch Zone Size
    colorAndOpacity {colorUseRule=UseColor_Specified, specifiedColor=<rgba 0.4125 0.4125 0.4125 1>}
    font {
     bForceMonospaced=false
     bMaterialIsStencil=false
     fontMaterial=null
     fontObject=&BurbankSmall
     letterSpacing=0
     monospacedWidth=1
     size=27
     skewAmount=0
     typefaceFontName=Bold
     outlineSettings {
      bApplyOutlineToDropShadows=false
      bMiteredCorners=false
      bSeparateFillAlpha=false
      outlineMaterial=null
      outlineSize=0
      outlineColor=<rgba 0 0 0 1>
     }
    }
    shadowOffset=<0 0>
   }
   slot {
    class=CanvasPanelSlot
    properties {layoutData {alignment=<0 0>, anchors {maximum=<0 0>, minimum=<0 0>}, offsets=<ltrb 28.8565 263.5549 543.0811 45>}}
   }
  }
  Percent2 {
   widgetClassPath=&CommonTextBlock
   parent=CanvasPanel_PropertiesPreview
   widget=Percent2
   properties {
    justification=Right
    scrollStyle=&TextScrollStyle-Base_C
    text=100%
    textTransformPolicy=ToUpper
    font {
     bForceMonospaced=false
     bMaterialIsStencil=false
     fontMaterial=null
     fontObject=&BurbankSmall
     letterSpacing=0
     monospacedWidth=1
     size=27
     skewAmount=0
     typefaceFontName=Bold
     outlineSettings {
      bApplyOutlineToDropShadows=false
      bMiteredCorners=false
      bSeparateFillAlpha=false
      outlineMaterial=null
      outlineSize=0
      outlineColor=<rgba 0 0 0 1>
     }
    }
    shadowOffset=<0 0>
   }
   slot {
    class=CanvasPanelSlot
    properties {layoutData {alignment=<1 0>, anchors {maximum=<1 0>, minimum=<1 0>}, offsets=<ltrb -126.5203 265.6425 313.2523 45>}}
   }
  }
  Image_Preview_1 {
   widgetClassPath=&Image
   bIsVariable=true
   parent=CanvasPanel_PropertiesPreview
   widget=Image_Preview_1
   properties {brush {
     bIsDynamicallyLoaded=false
     drawAs=Image
     imageType=NoImage
     mirroring=NoMirror
     resourceName=None
     resourceObject=/Game/Athena/HUD/Bacchus/HUDLayoutTool/HLT_Icons/T-HLT-Tutorial-NoText_PropertiesPreview
     tiling=NoTile
     imageSize=<953.82 437.22>
     margin=<ltrb 0 0 0 0>
     outlineSettings {
      bUseBrushTransparency=false
      roundingType=HalfHeightRadius
      width=0
      color {colorUseRule=UseColor_Specified, specifiedColor=<rgba 0 0 0 0>}
      cornerRadii=<0 0 0 0>
     }
     tintColor {colorUseRule=UseColor_Specified, specifiedColor=<rgba 1 1 1 1>}
     uVRegion {bIsValid=false, max=<0 0>, min=<0 0>}
    }}
   slot {
    class=CanvasPanelSlot
    properties {bAutoSize=true, layoutData {alignment=<0 0>, anchors {maximum=<0 0>, minimum=<0 0>}, offsets=<ltrb 0 0 953.82 437.22>}}
   }
  }
  PreviewOutline_2 {
   widgetClassPath=&Border
   parent=Overlay_PropertiesPreview
   widget=PreviewOutline_2
   properties {
    background {
     bIsDynamicallyLoaded=false
     drawAs=Box
     imageType=NoImage
     mirroring=NoMirror
     resourceName=None
     resourceObject=&M_HUDLayoutToolV2_Tutorial_PreviewOutline
     tiling=NoTile
     imageSize=<32 32>
     margin=<ltrb 0.5 0.5 0.5 0.5>
     outlineSettings {
      bUseBrushTransparency=false
      roundingType=HalfHeightRadius
      width=0
      color {colorUseRule=UseColor_Specified, specifiedColor=<rgba 0 0 0 0>}
      cornerRadii=<0 0 0 0>
     }
     tintColor {colorUseRule=UseColor_Specified, specifiedColor=<rgba 1 1 1 1>}
     uVRegion {bIsValid=false, max=<0 0>, min=<0 0>}
    }
    brushColor=<rgba 0.9301 1 0.0091 1>
   }
   slot {class=OverlaySlot, properties {horizontalAlignment=HAlign_Fill, verticalAlignment=VAlign_Fill, padding=<ltrb -5 -7 -7 -8>}}
  }
  HUDLayoutToolV2_TutorialTextPrompt_2 {
   widgetClassPath=&HUDLayoutToolV2_TutorialTextPrompt
   bIsVariable=true
   parent=Overlay3_Properties
   widget=HUDLayoutToolV2_TutorialTextPrompt_2
   properties {
    descriptionText="Customize your buttons and HUD elements!"
    titleText="Properties!"
    renderTransform {angle=3, scale=<1 1>, shear=<0 0>, translation=<0 0>}
   }
   slot {class=OverlaySlot, properties {horizontalAlignment=HAlign_Center, padding=<ltrb 540 380 0 0>}}
  }
  Overlay4_Catalog {
   widgetClassPath=&Overlay
   parent=Switcher_TutorialOverlays
   widget=Overlay4_Catalog
   properties {visibility=Visible}
   slot {class=CommonVisibilitySwitcherSlot}
  }
  SafeZone_0 {
   widgetClassPath=&SafeZone
   bIsVariable=true
   parent=Overlay4_Catalog
   widget=SafeZone_0
   properties {padBottom=false, padLeft=false, padRight=false, padTop=false}
   slot {class=OverlaySlot, properties {horizontalAlignment=HAlign_Fill, verticalAlignment=VAlign_Bottom}}
  }
  Overlay_ContextPreview_1 {widgetClassPath=&Overlay, parent=SafeZone_0, widget=Overlay_ContextPreview_1, slot {class=SafeZoneSlot}}
  PreviewOutline_3 {
   widgetClassPath=&Border
   parent=Overlay_ContextPreview_1
   widget=PreviewOutline_3
   properties {
    background {
     bIsDynamicallyLoaded=false
     drawAs=Box
     imageType=NoImage
     mirroring=NoMirror
     resourceName=None
     resourceObject=&M_HUDLayoutToolV2_Tutorial_PreviewOutline
     tiling=NoTile
     imageSize=<32 32>
     margin=<ltrb 0.4 0.36 0.4 0.5>
     outlineSettings {
      bUseBrushTransparency=false
      roundingType=HalfHeightRadius
      width=0
      color {colorUseRule=UseColor_Specified, specifiedColor=<rgba 0 0 0 0>}
      cornerRadii=<0 0 0 0>
     }
     tintColor {colorUseRule=UseColor_Specified, specifiedColor=<rgba 1 1 1 1>}
     uVRegion {bIsValid=false, max=<0 0>, min=<0 0>}
    }
    brushColor=<rgba 0.9301 1 0.0091 1>
    padding=<ltrb 472 174 472 174>
   }
   slot {class=OverlaySlot, properties {horizontalAlignment=HAlign_Fill, verticalAlignment=VAlign_Fill}}
  }
  Border_NewTag_2 {
   widgetClassPath=&CommonBorder
   bIsVariable=true
   parent=Overlay_ContextPreview_1
   widget=Border_NewTag_2
   properties {
    bFlipForRightToLeftFlowDirection=true
    visibility=Collapsed
    background {
     bIsDynamicallyLoaded=false
     drawAs=Box
     imageType=NoImage
     mirroring=NoMirror
     resourceName=None
     resourceObject=&T-HLT-Tutorial_NewTag_Sprite
     tiling=NoTile
     imageSize=<128 64>
     margin=<ltrb 0.1 0 0.25 0>
     outlineSettings {
      bUseBrushTransparency=false
      roundingType=HalfHeightRadius
      width=0
      color {colorUseRule=UseColor_Specified, specifiedColor=<rgba 0 0 0 0>}
      cornerRadii=<0 0 0 0>
     }
     tintColor {colorUseRule=UseColor_Specified, specifiedColor=<rgba 1 1 1 1>}
     uVRegion {bIsValid=false, max=<0 0>, min=<0 0>}
    }
    padding=<ltrb 26 14 40 10>
    renderTransform {angle=0, scale=<1.2 1.2>, shear=<0 0>, translation=<0 0>}
   }
   slot {class=OverlaySlot, properties {horizontalAlignment=HAlign_Right, padding=<ltrb 0 -45 150 0>}}
  }
  SizeBox_1 {
   widgetClassPath=&SizeBox
   parent=Border_NewTag_2
   widget=SizeBox_1
   properties {bOverride_MaxDesiredWidth=true, maxDesiredWidth=256}
   slot {class=BorderSlot, properties {padding=<ltrb 26 14 40 10>}}
  }
  Text_Tag_2 {
   widgetClassPath=&CommonTextBlock
   bIsVariable=true
   parent=SizeBox_1
   widget=Text_Tag_2
   properties {
    text="New!"
    textTransformPolicy=ToUpper
    font {
     bForceMonospaced=false
     bMaterialIsStencil=false
     fontMaterial=null
     fontObject=&Burbank
     letterSpacing=0
     monospacedWidth=1
     size=36
     skewAmount=0
     typefaceFontName=Burbank-Black
     outlineSettings {
      bApplyOutlineToDropShadows=false
      bMiteredCorners=false
      bSeparateFillAlpha=false
      outlineMaterial=null
      outlineSize=0
      outlineColor=<rgba 0 0 0 1>
     }
    }
    renderTransform {angle=0, scale=<1 0.86>, shear=<0 0>, translation=<0 0>}
    shadowOffset=<0 0>
   }
   slot {class=SizeBoxSlot}
  }
  Image_Arrow_3 {
   widgetClassPath=&Image
   bIsVariable=true
   parent=Overlay4_Catalog
   widget=Image_Arrow_3
   properties {
    brush {
     bIsDynamicallyLoaded=false
     drawAs=Image
     imageType=NoImage
     mirroring=NoMirror
     resourceName=None
     resourceObject=&T-HLT-Tutorial_Arrow_Sprite
     tiling=NoTile
     imageSize=<392 800>
     margin=<ltrb 0 0 0 0>
     outlineSettings {
      bUseBrushTransparency=false
      roundingType=HalfHeightRadius
      width=0
      color {colorUseRule=UseColor_Specified, specifiedColor=<rgba 0 0 0 0>}
      cornerRadii=<0 0 0 0>
     }
     tintColor {colorUseRule=UseColor_Specified, specifiedColor=<rgba 1 1 1 1>}
     uVRegion {bIsValid=false, max=<0 0>, min=<0 0>}
    }
    renderTransform {angle=-90, scale=<-1 1>, shear=<0 0>, translation=<0 0>}
   }
   slot {class=OverlaySlot, properties {horizontalAlignment=HAlign_Center, verticalAlignment=VAlign_Bottom, padding=<ltrb 0 0 440 90>}}
  }
  HUDLayoutToolV2_TutorialTextPrompt_3 {
   widgetClassPath=&HUDLayoutToolV2_TutorialTextPrompt
   bIsVariable=true
   parent=Overlay4_Catalog
   widget=HUDLayoutToolV2_TutorialTextPrompt_3
   properties {
    descriptionText="Place your buttons and HUD elements!"
    titleText="HUD Toolbox!"
    renderTransform {angle=-3, scale=<1 1>, shear=<0 0>, translation=<0 0>}
   }
   slot {class=OverlaySlot, properties {horizontalAlignment=HAlign_Center, verticalAlignment=VAlign_Bottom, padding=<ltrb 0 0 60 480>}}
  }
  Button_NextTutorial {
   widgetClassPath=/Game/Athena/HUD/Bacchus/HUDLayoutTool/HUDLayoutToolV2_Button
   bInherited=true
   bIsVariable=true
   parent=Overlay_197
   widget=Button_NextTutorial
   properties {colorAndOpacity=<rgba 1 1 1 0>}
   slot {class=OverlaySlot, properties {horizontalAlignment=HAlign_Fill, verticalAlignment=VAlign_Fill}}
  }
 ]
}
