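# Warning popup widget for the HUD Layout Tool, with default text asking to confirm a reset to the last saved layout.
# A blurred, darkened backdrop frames a panel with warning icon, title, description and Confirm, Cancel and Close buttons.
# Anim_In pops the panel, spins in the icon and staggers text and buttons; Anim_Out shrinks and fades it.
# The event graph is empty; key widgets are inherited from native HUDLayoutToolV2_Popup.
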WidgetBlueprint /FortniteCoreUI/Athena/HUD/Bacchus/HUDLayoutTool/HUDLayoutToolV2_WarningPopup
animations [
 Anim_In {
  binding_count=9
  length_seconds=1.5489
  object=&self:Anim_In
  sequence {bindings [
    RichText_Description {
     id=#1
     possessed_class=RichTextBlock
     tracks [
      {
       type=MovieSceneFloatTrack
       display_name=Render Opacity
       sections [{
         type=MovieSceneFloatSection
         end_frame=24
         start_frame=-1
         channels [None {type=MovieSceneScriptingFloatChannel, keys [[1,0],[5.685,0],[9.269,1]]}]
        }]
      }
      {
       type=MovieScene2DTransformTrack
       sections [{
         type=MovieScene2DTransformSection
         end_frame=24
         start_frame=-1
         channels [
          Translation.X {type=MovieSceneScriptingFloatChannel, keys [[1,260],[4,260],[9.269,0]]}
          Translation.Y {type=MovieSceneScriptingFloatChannel}
          Angle {type=MovieSceneScriptingFloatChannel}
          Scale.X {type=MovieSceneScriptingFloatChannel}
          Scale.Y {type=MovieSceneScriptingFloatChannel}
          Shear.X {type=MovieSceneScriptingFloatChannel}
          Shear.Y {type=MovieSceneScriptingFloatChannel}
         ]
        }]
      }
     ]
    }
    PopupOverlay {
     id=#2
     possessed_class=Overlay
     tracks [{
       type=MovieSceneFloatTrack
       display_name=Render Opacity
       sections [{type=MovieSceneFloatSection, end_frame=0, start_frame=0, channels [None {type=MovieSceneScriptingFloatChannel, const [1,1]}]}]
      }]
    }
    Button_Confirm {
     id=#3
     possessed_class=HUDLayoutToolV2_Button_C
     tracks [
      {
       type=MovieSceneFloatTrack
       display_name=Render Opacity
       sections [{
         type=MovieSceneFloatSection
         end_frame=24
         start_frame=0
         channels [None {type=MovieSceneScriptingFloatChannel, keys [[0,0],[10,0],[13.602,1]]}]
        }]
      }
      {
       type=MovieScene2DTransformTrack
       sections [{
         type=MovieScene2DTransformSection
         end_frame=30
         start_frame=0
         channels [
          Translation.X {type=MovieSceneScriptingFloatChannel}
          Translation.Y {type=MovieSceneScriptingFloatChannel}
          Angle {type=MovieSceneScriptingFloatChannel}
          Scale.X {type=MovieSceneScriptingFloatChannel, keys [[9.6023,1.2],[13.6023,0.92],[17.6023,1.01],[22.6023,1]]}
          Scale.Y {type=MovieSceneScriptingFloatChannel, keys [[9.6023,1.2],[13.6023,0.92],[17.6023,1.01],[22.6023,1]]}
          Shear.X {type=MovieSceneScriptingFloatChannel}
          Shear.Y {type=MovieSceneScriptingFloatChannel}
         ]
        }]
      }
     ]
    }
    Button_Cancel {
     id=#4
     possessed_class=HUDLayoutToolV2_Button_C
     tracks [
      {
       type=MovieSceneFloatTrack
       display_name=Render Opacity
       sections [{
         type=MovieSceneFloatSection
         end_frame=24
         start_frame=0
         channels [None {type=MovieSceneScriptingFloatChannel, keys [[0,0],[11,0],[14.602,1]]}]
        }]
      }
      {
       type=MovieScene2DTransformTrack
       sections [{
         type=MovieScene2DTransformSection
         end_frame=24
         start_frame=0
         channels [
          Translation.X {type=MovieSceneScriptingFloatChannel}
          Translation.Y {type=MovieSceneScriptingFloatChannel}
          Angle {type=MovieSceneScriptingFloatChannel}
          Scale.X {type=MovieSceneScriptingFloatChannel, keys [[0,1],[10.6023,1.2],[14.6023,0.92],[18.6023,1.01],[23.6023,1]]}
          Scale.Y {type=MovieSceneScriptingFloatChannel, keys [[0,1],[10.6023,1.2],[14.6023,0.92],[18.6023,1.01],[23.6023,1]]}
          Shear.X {type=MovieSceneScriptingFloatChannel}
          Shear.Y {type=MovieSceneScriptingFloatChannel}
         ]
        }]
      }
     ]
    }
    CommonText_Title {
     id=#5
     possessed_class=CommonTextBlock
     tracks [
      {
       type=MovieSceneFloatTrack
       display_name=Render Opacity
       sections [{
         type=MovieSceneFloatSection
         end_frame=26
         start_frame=0
         channels [None {type=MovieSceneScriptingFloatChannel, keys [[0,0],[4.685,0],[8.269,1]]}]
        }]
      }
      {
       type=MovieScene2DTransformTrack
       sections [{
         type=MovieScene2DTransformSection
         end_frame=25
         start_frame=0
         channels [
          Translation.X {type=MovieSceneScriptingFloatChannel, keys [[0,260],[3,260],[8.269,0]]}
          Translation.Y {type=MovieSceneScriptingFloatChannel}
          Angle {type=MovieSceneScriptingFloatChannel}
          Scale.X {type=MovieSceneScriptingFloatChannel}
          Scale.Y {type=MovieSceneScriptingFloatChannel}
          Shear.X {type=MovieSceneScriptingFloatChannel}
          Shear.Y {type=MovieSceneScriptingFloatChannel}
         ]
        }]
      }
     ]
    }
    Border_Darken {
     id=#6
     possessed_class=Border
     tracks [{
       type=MovieSceneFloatTrack
       display_name=Render Opacity
       sections [{type=MovieSceneFloatSection, end_frame=24, start_frame=0, channels [None {type=MovieSceneScriptingFloatChannel, keys [[0,0],[4,1],[6,1]]}]}]
      }]
    }
    BackgroundBlur_1 {
     id=#7
     possessed_class=BackgroundBlur
     tracks [
      {
       type=MovieSceneFloatTrack
       display_name=Blur Strength
       sections [{type=MovieSceneFloatSection, end_frame=24, start_frame=0, channels [None {type=MovieSceneScriptingFloatChannel, keys [[0,0],[4,20],[6,20]]}]}]
      }
      {
       type=MovieSceneFloatTrack
       display_name=Render Opacity
       sections [{type=MovieSceneFloatSection, end_frame=24, start_frame=0, channels [None {type=MovieSceneScriptingFloatChannel, keys [[0,0],[4,1],[6,1]]}]}]
      }
     ]
    }
    SizeBox_1 {
     id=#8
     possessed_class=SizeBox
     tracks [{
       type=MovieScene2DTransformTrack
       sections [{
         type=MovieScene2DTransformSection
         end_frame=24
         start_frame=0
         channels [
          Translation.X {type=MovieSceneScriptingFloatChannel}
          Translation.Y {type=MovieSceneScriptingFloatChannel}
          Angle {type=MovieSceneScriptingFloatChannel}
          Scale.X {type=MovieSceneScriptingFloatChannel, keys [[0,0],[3,1.15],[6,1]]}
          Scale.Y {type=MovieSceneScriptingFloatChannel, keys [[0,0],[3,1.15],[6,1]]}
          Shear.X {type=MovieSceneScriptingFloatChannel}
          Shear.Y {type=MovieSceneScriptingFloatChannel}
         ]
        }]
      }]
    }
    Image_Warning {
     id=#9
     possessed_class=Image
     tracks [
      {
       type=MovieSceneFloatTrack
       display_name=Render Opacity
       sections [{
         type=MovieSceneFloatSection
         end_frame=24
         start_frame=-1
         channels [None {type=MovieSceneScriptingFloatChannel, keys [[0,0],[6,0],[6.8383,0.01],[9.484,1]]}]
        }]
      }
      {
       type=MovieScene2DTransformTrack
       sections [{
         type=MovieScene2DTransformSection
         end_frame=24
         start_frame=-1
         channels [
          Translation.X {type=MovieSceneScriptingFloatChannel}
          Translation.Y {type=MovieSceneScriptingFloatChannel}
          Angle {type=MovieSceneScriptingFloatChannel, keys [[6,-180],[9.484,-4]]}
          Scale.X {type=MovieSceneScriptingFloatChannel, keys [[6,4],[9.484,0.75],[11.8067,1.25],[15.2907,1]]}
          Scale.Y {type=MovieSceneScriptingFloatChannel, keys [[6,4],[9.484,0.75],[11.8003,1.5],[11.8067,1.25],[15.2907,1]]}
          Shear.X {type=MovieSceneScriptingFloatChannel}
          Shear.Y {type=MovieSceneScriptingFloatChannel}
         ]
        }]
      }
     ]
    }
   ]}
 }
 Anim_Out {
  binding_count=4
  length_seconds=0.35
  object=&self:Anim_Out
  sequence {bindings [
    Border_Darken {
     id=#6
     possessed_class=Border
     tracks [{
       type=MovieSceneFloatTrack
       display_name=Render Opacity
       sections [{type=MovieSceneFloatSection, end_frame=7, start_frame=0, channels [None {type=MovieSceneScriptingFloatChannel, keys [[0,1],[3,1],[6,0]]}]}]
      }]
    }
    BackgroundBlur_1 {
     id=#7
     possessed_class=BackgroundBlur
     tracks [
      {
       type=MovieSceneFloatTrack
       display_name=Blur Strength
       sections [{type=MovieSceneFloatSection, end_frame=7, start_frame=0, channels [None {type=MovieSceneScriptingFloatChannel, keys [[0,20],[3,20],[6,0]]}]}]
      }
      {
       type=MovieSceneFloatTrack
       display_name=Render Opacity
       sections [{type=MovieSceneFloatSection, end_frame=7, start_frame=0, channels [None {type=MovieSceneScriptingFloatChannel, keys [[0,1],[3,1],[6,0]]}]}]
      }
     ]
    }
    PopupOverlay {
     id=#10
     possessed_class=Overlay
     tracks [{
       type=MovieSceneFloatTrack
       display_name=Render Opacity
       sections [{type=MovieSceneFloatSection, end_frame=6, start_frame=0, channels [None {type=MovieSceneScriptingFloatChannel, keys [[0,1],[5,1],[6,0]]}]}]
      }]
    }
    SizeBox_1 {
     id=#8
     possessed_class=SizeBox
     tracks [{
       type=MovieScene2DTransformTrack
       sections [{
         type=MovieScene2DTransformSection
         end_frame=6
         start_frame=0
         channels [
          Translation.X {type=MovieSceneScriptingFloatChannel}
          Translation.Y {type=MovieSceneScriptingFloatChannel}
          Angle {type=MovieSceneScriptingFloatChannel}
          Scale.X {type=MovieSceneScriptingFloatChannel, keys [[0,1],[3,1.15],[6,0.1]]}
          Scale.Y {type=MovieSceneScriptingFloatChannel, keys [[0,1],[3,1.15],[6,0.1]]}
          Shear.X {type=MovieSceneScriptingFloatChannel}
          Shear.Y {type=MovieSceneScriptingFloatChannel}
         ]
        }]
      }]
    }
   ]}
 }
]
graphs [EventGraph {links [], nodes []}]
variables []
widgets {
 info {inheritedWidgetCount=6, parentClass=/Script/FortniteUI.HUDLayoutToolV2_Popup, rootWidgetClass=&Overlay, widgetCount=24}
 widgets [
  Overlay_1 {widgetClassPath=&Overlay, widget=Overlay_1}
  BackgroundBlur_1 {
   widgetClassPath=&BackgroundBlur
   parent=Overlay_1
   widget=BackgroundBlur_1
   properties {
    blurStrength=20
    lowQualityFallbackBrush {
     bIsDynamicallyLoaded=false
     drawAs=Image
     imageType=NoImage
     mirroring=NoMirror
     resourceName=None
     resourceObject=/Game/UI/Foundation/Textures/LoadingScreens/BacchusTutorial/LoadingScreenBG01
     tiling=NoTile
     imageSize=<1024 768>
     margin=<ltrb 0 0 0 0>
     outlineSettings {
      bUseBrushTransparency=false
      roundingType=HalfHeightRadius
      width=0
      color {colorUseRule=UseColor_Specified, specifiedColor=<rgba 0 0 0 0>}
      cornerRadii=<0 0 0 0>
     }
     tintColor {colorUseRule=UseColor_Specified, specifiedColor=<rgba 0 0 0 0.33>}
     uVRegion {bIsValid=false, max=<0 0>, min=<0 0>}
    }
   }
   slot {class=OverlaySlot, properties {horizontalAlignment=HAlign_Fill, verticalAlignment=VAlign_Fill}}
  }
  Border_Darken {
   widgetClassPath=&Border
   parent=Overlay_1
   widget=Border_Darken
   properties {brushColor=<rgba 0.0052 0.0052 0.0056 0.1571>, padding=<ltrb 0 0 0 0>}
   slot {class=OverlaySlot, properties {horizontalAlignment=HAlign_Fill, verticalAlignment=VAlign_Fill}}
  }
  SizeBox_1 {
   widgetClassPath=&SizeBox
   parent=Overlay_1
   widget=SizeBox_1
   properties {bOverride_MinDesiredHeight=true, bOverride_WidthOverride=true, heightOverride=476, minDesiredHeight=476, widthOverride=1441}
   slot {class=OverlaySlot, properties {horizontalAlignment=HAlign_Center, verticalAlignment=VAlign_Center}}
  }
  PopupOverlay {widgetClassPath=&Overlay, parent=SizeBox_1, widget=PopupOverlay, slot {class=SizeBoxSlot}}
  Background {
   widgetClassPath=&Image
   parent=PopupOverlay
   widget=Background
   properties {colorAndOpacity=<rgba 0 0.0222 0.2874 1>}
   slot {class=OverlaySlot, properties {horizontalAlignment=HAlign_Fill, verticalAlignment=VAlign_Fill}}
  }
  Image_Header {
   widgetClassPath=&Image
   parent=PopupOverlay
   widget=Image_Header
   properties {
    brush {
     bIsDynamicallyLoaded=false
     drawAs=Image
     imageType=NoImage
     mirroring=NoMirror
     resourceName=None
     resourceObject=/Game/Athena/HUD/Bacchus/HUDLayoutTool/HLT_Icons/M_HLT_ManagementHeader
     tiling=NoTile
     imageSize=<32 146>
     margin=<ltrb 0 0 0 0>
     outlineSettings {
      bUseBrushTransparency=false
      roundingType=HalfHeightRadius
      width=0
      color {colorUseRule=UseColor_Specified, specifiedColor=<rgba 0 0 0 0>}
      cornerRadii=<0 0 0 0>
     }
     tintColor {colorUseRule=UseColor_Specified, specifiedColor=<rgba 1 1 1 1>}
     uVRegion {bIsValid=false, max=<0 0>, min=<0 0>}
    }
    renderTransform {angle=0, scale=<-1 1>, shear=<0 0>, translation=<0 0>}
   }
   slot {class=OverlaySlot, properties {horizontalAlignment=HAlign_Fill}}
  }
  Image_Footer {
   widgetClassPath=&Image
   parent=PopupOverlay
   widget=Image_Footer
   properties {
    brush {
     bIsDynamicallyLoaded=false
     drawAs=Image
     imageType=NoImage
     mirroring=NoMirror
     resourceName=None
     resourceObject=/Game/Athena/HUD/Bacchus/HUDLayoutTool/HLT_Icons/M_HLT_PopupBottomBar
     tiling=NoTile
     imageSize=<32 74>
     margin=<ltrb 0 0 0 0>
     outlineSettings {
      bUseBrushTransparency=false
      roundingType=HalfHeightRadius
      width=0
      color {colorUseRule=UseColor_Specified, specifiedColor=<rgba 0 0 0 0>}
      cornerRadii=<0 0 0 0>
     }
     tintColor {colorUseRule=UseColor_Specified, specifiedColor=<rgba 1 1 1 1>}
     uVRegion {bIsValid=false, max=<0 0>, min=<0 0>}
    }
    renderTransform {angle=0, scale=<1 -1>, shear=<0 0>, translation=<0 0>}
   }
   slot {class=OverlaySlot, properties {horizontalAlignment=HAlign_Fill, verticalAlignment=VAlign_Bottom}}
  }
  VerticalBox_0 {
   widgetClassPath=&VerticalBox
   parent=PopupOverlay
   widget=VerticalBox_0
   slot {class=OverlaySlot, properties {horizontalAlignment=HAlign_Fill, verticalAlignment=VAlign_Fill}}
  }
  SizeBox_0 {
   widgetClassPath=&SizeBox
   parent=VerticalBox_0
   widget=SizeBox_0
   properties {bOverride_HeightOverride=true, heightOverride=146}
   slot {class=VerticalBoxSlot}
  }
  Overlay_0 {widgetClassPath=&Overlay, parent=SizeBox_0, widget=Overlay_0, slot {class=SizeBoxSlot}}
  HorizontalBox_0 {
   widgetClassPath=&HorizontalBox
   parent=Overlay_0
   widget=HorizontalBox_0
   slot {class=OverlaySlot, properties {padding=<ltrb 0 20 0 0>}}
  }
  Switcher_CornerIcon {
   widgetClassPath=&CommonActivatableWidgetSwitcher
   bInherited=true
   parent=HorizontalBox_0
   widget=Switcher_CornerIcon
   slot {class=HorizontalBoxSlot, properties {horizontalAlignment=HAlign_Left, verticalAlignment=VAlign_Center, size {sizeRule=Fill, value=1}}}
  }
  Image_Warning {
   widgetClassPath=&Image
   parent=Switcher_CornerIcon
   widget=Image_Warning
   properties {
    brush {
     bIsDynamicallyLoaded=false
     drawAs=Image
     imageType=NoImage
     mirroring=NoMirror
     resourceName=None
     resourceObject=/Game/UI/Foundation/Shell/Textures/T-Icon-Warning-128
     tiling=NoTile
     imageSize=<160 160>
     margin=<ltrb 0 0 0 0>
     outlineSettings {
      bUseBrushTransparency=false
      roundingType=HalfHeightRadius
      width=0
      color {colorUseRule=UseColor_Specified, specifiedColor=<rgba 0 0 0 0>}
      cornerRadii=<0 0 0 0>
     }
     tintColor {colorUseRule=UseColor_Specified, specifiedColor=<rgba 1 1 1 1>}
     uVRegion {bIsValid=false, max=<0 0>, min=<0 0>}
    }
    colorAndOpacity=<rgba 0.9301 1 0.0097 1>
    renderTransform {angle=-6, scale=<1 1>, shear=<0 0>, translation=<-20 -58>}
   }
   slot {class=WidgetSwitcherSlot, properties {padding=<ltrb 0 0 0 -40>}}
  }
  Spacer_88 {
   widgetClassPath=&Spacer
   parent=Switcher_CornerIcon
   widget=Spacer_88
   properties {visibility=Collapsed, size=<50 50>}
   slot {class=WidgetSwitcherSlot}
  }
  CommonText_Title {
   widgetClassPath=&CommonTextBlock
   bInherited=true
   parent=HorizontalBox_0
   widget=CommonText_Title
   properties {
    text=Reset HUD layout?
    textTransformPolicy=ToUpper
    font {
     bForceMonospaced=false
     bMaterialIsStencil=false
     fontMaterial=null
     fontObject=&BurbankSmall
     letterSpacing=0
     monospacedWidth=1
     size=46
     skewAmount=0
     typefaceFontName=Black
     outlineSettings {
      bApplyOutlineToDropShadows=false
      bMiteredCorners=false
      bSeparateFillAlpha=false
      outlineMaterial=null
      outlineSize=0
      outlineColor=<rgba 0 0 0 1>
     }
    }
    renderTransform {angle=0, scale=<1 1>, shear=<-10 0>, translation=<0 0>}
    shadowOffset=<0 0>
   }
   slot {class=HorizontalBoxSlot, properties {padding=<ltrb 0 10 0 0>}}
  }
  Spacer_156 {
   widgetClassPath=&Spacer
   parent=VerticalBox_0
   widget=Spacer_156
   properties {size=<20 20>}
   slot {class=VerticalBoxSlot, properties {size {sizeRule=Fill, value=1}}}
  }
  RichText_Description {
   widgetClassPath=/Script/UMG.RichTextBlock
   bInherited=true
   parent=VerticalBox_0
   widget=RichText_Description
   properties {
    autoWrapText=true
    justification=Center
    text="Are you sure you want to <Highlight>RESET</> your custom HUD Layout to its <Highlight>LAST SAVED STATE?</>"
    textStyleSet=/Game/Athena/HUD/Bacchus/HUDLayoutTool/Styles/HUDLayoutToolV2_CalloutStyles
    defaultTextStyleOverride {
     overflowPolicy=Clip
     transformPolicy=None
     colorAndOpacity {colorUseRule=UseColor_Specified, specifiedColor=<rgba 0.2195 0.6584 1 1>}
     font {
      bForceMonospaced=false
      bMaterialIsStencil=false
      fontMaterial=null
      fontObject=&BurbankSmall
      letterSpacing=0
      monospacedWidth=1
      size=24
      skewAmount=0
      typefaceFontName=Black
      outlineSettings {
       bApplyOutlineToDropShadows=false
       bMiteredCorners=false
       bSeparateFillAlpha=false
       outlineMaterial=null
       outlineSize=0
       outlineColor=<rgba 0 0 0 1>
      }
     }
     highlightColor {colorUseRule=UseColor_Specified, specifiedColor=<rgba 0 0 0 1>}
     highlightShape {
      bIsDynamicallyLoaded=false
      drawAs=Image
      imageType=NoImage
      mirroring=NoMirror
      resourceName=None
      resourceObject=null
      tiling=NoTile
      imageSize=<32 32>
      margin=<ltrb 0 0 0 0>
      outlineSettings {
       bUseBrushTransparency=false
       roundingType=HalfHeightRadius
       width=0
       color {colorUseRule=UseColor_Specified, specifiedColor=<rgba 0 0 0 0>}
       cornerRadii=<0 0 0 0>
      }
      tintColor {colorUseRule=UseColor_Specified, specifiedColor=<rgba 1 1 1 1>}
      uVRegion {bIsValid=false, max=<0 0>, min=<0 0>}
     }
     selectedBackgroundColor {colorUseRule=UseColor_Foreground, specifiedColor=<rgba 1 0 1 1>}
     shadowColorAndOpacity=<rgba 0 0 0 1>
     shadowOffset=<0 0>
     strikeBrush {
      bIsDynamicallyLoaded=false
      drawAs=Image
      imageType=NoImage
      mirroring=NoMirror
      resourceName=None
      resourceObject=null
      tiling=NoTile
      imageSize=<32 32>
      margin=<ltrb 0 0 0 0>
      outlineSettings {
       bUseBrushTransparency=false
       roundingType=HalfHeightRadius
       width=0
       color {colorUseRule=UseColor_Specified, specifiedColor=<rgba 0 0 0 0>}
       cornerRadii=<0 0 0 0>
      }
      tintColor {colorUseRule=UseColor_Specified, specifiedColor=<rgba 1 1 1 1>}
      uVRegion {bIsValid=false, max=<0 0>, min=<0 0>}
     }
     underlineBrush {
      bIsDynamicallyLoaded=false
      drawAs=Image
      imageType=NoImage
      mirroring=NoMirror
      resourceName=None
      resourceObject=null
      tiling=NoTile
      imageSize=<32 32>
      margin=<ltrb 0 0 0 0>
      outlineSettings {
       bUseBrushTransparency=false
       roundingType=HalfHeightRadius
       width=0
       color {colorUseRule=UseColor_Specified, specifiedColor=<rgba 0 0 0 0>}
       cornerRadii=<0 0 0 0>
      }
      tintColor {colorUseRule=UseColor_Specified, specifiedColor=<rgba 1 1 1 1>}
      uVRegion {bIsValid=false, max=<0 0>, min=<0 0>}
     }
    }
   }
   slot {class=VerticalBoxSlot, properties {padding=<ltrb 120 0 120 0>}}
  }
  Spacer {
   widgetClassPath=&Spacer
   parent=VerticalBox_0
   widget=Spacer
   properties {size=<20 20>}
   slot {class=VerticalBoxSlot, properties {size {sizeRule=Fill, value=1}}}
  }
  Overlay_252 {
   widgetClassPath=&Overlay
   parent=VerticalBox_0
   widget=Overlay_252
   slot {class=VerticalBoxSlot, properties {horizontalAlignment=HAlign_Right, verticalAlignment=VAlign_Bottom, padding=<ltrb 50 0 20 20>}}
  }
  HorizontalBox_248 {
   widgetClassPath=&HorizontalBox
   parent=Overlay_252
   widget=HorizontalBox_248
   slot {class=OverlaySlot, properties {horizontalAlignment=HAlign_Right, verticalAlignment=VAlign_Center}}
  }
  Button_Confirm {
   widgetClassPath=/Game/Athena/HUD/Bacchus/HUDLayoutTool/HUDLayoutToolV2_Button
   bInherited=true
   parent=HorizontalBox_248
   widget=Button_Confirm
   properties {
    bOverride_HeightOverride=true
    bOverride_WidthOverride=true
    buttonHeightOverride=80
    buttonWidthOverride=246
    text=Reset
    textFontSize=26
    internalPadding=<ltrb 24 12 24 12>
   }
   slot {class=HorizontalBoxSlot}
  }
  Button_Cancel {
   widgetClassPath=&HUDLayoutToolV2_Button
   bInherited=true
   parent=HorizontalBox_248
   widget=Button_Cancel
   properties {
    bOverride_HeightOverride=true
    bOverride_WidthOverride=true
    buttonHeightOverride=80
    buttonWidthOverride=246
    text=Cancel
    textFontSize=26
    internalPadding=<ltrb 24 12 24 12>
    normalButtonStyle {
     iconAngle=0
     backgroundColor=<rgba 0.9301 1 0.0097 1>
     iconColor=<rgba 0 0 0 1>
     textColor {colorUseRule=UseColor_Specified, specifiedColor=<rgba 0 0 0 1>}
    }
    pressedButtonStyle {
     iconAngle=0
     backgroundColor=<rgba 1 1 1 1>
     iconColor=<rgba 0 0 0 1>
     textColor {colorUseRule=UseColor_Specified, specifiedColor=<rgba 0 0 0 1>}
    }
   }
   slot {class=HorizontalBoxSlot, properties {horizontalAlignment=HAlign_Right, verticalAlignment=VAlign_Center, padding=<ltrb 20 0 0 0>}}
  }
  Button_Close {
   widgetClassPath=/FortUILibrary/Foundation/Buttons/CloseButton
   bInherited=true
   parent=PopupOverlay
   widget=Button_Close
   properties {
    onButtonBaseClicked=(&self:WidgetTree.Button_Close.HandleOnButtonClicked)
    onButtonBaseDoubleClicked=(&self:WidgetTree.Button_Close.HandleOnButtonDoubleClicked)
    onButtonBaseHovered=(&self:WidgetTree.Button_Close.HandleOnButtonHovered)
    onButtonBaseUnhovered=(&self:WidgetTree.Button_Close.HandleOnButtonUnhovered)
    onSelectedChangedBase=(&self:WidgetTree.Button_Close.HandleOnSelectedChanged)
   }
   slot {class=OverlaySlot, properties {horizontalAlignment=HAlign_Right, padding=<ltrb 0 20 20 0>}}
  }
 ]
}
