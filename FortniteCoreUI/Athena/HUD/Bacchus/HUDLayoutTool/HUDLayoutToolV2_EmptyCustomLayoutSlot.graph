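# Placeholder widget for an empty custom layout slot in the Athena HUD Layout Tool V2, built on CommonUserWidget.
# A root SizeBox with fixed width and height overrides wraps a single Border tinted a faint, mostly transparent dark blue.
# It is purely visual, with an empty event graph and no variables.
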WidgetBlueprint /FortniteCoreUI/Athena/HUD/Bacchus/HUDLayoutTool/HUDLayoutToolV2_EmptyCustomLayoutSlot
graphs [EventGraph {links [], nodes []}]
variables []
widgets {
 info {inheritedWidgetCount=0, parentClass=&CommonUserWidget, rootWidgetClass=&SizeBox, widgetCount=2}
 widgets [
  SizeBox_Control {
   widgetClassPath=&SizeBox
   widget=SizeBox_Control
   properties {bOverride_HeightOverride=true, bOverride_WidthOverride=true, heightOverride=120, widthOverride=530}
  }
  Border_0 {
   widgetClassPath=&Border
   parent=SizeBox_Control
   widget=Border_0
   properties {brushColor=<rgba 0 0.0014 0.026 0.1>}
   slot {class=SizeBoxSlot}
  }
 ]
}
