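# Widget Blueprint for a toggleable layout-selection button in the Bacchus HUD Layout Tool V2, built on the native HUDLayoutToolV2_LayoutButton class.
# An overlay stacks the inherited Button_Internal, showing the layout name, beneath a second overlay holding two corner badges.
# The "Active" and "Legacy" badges are inherited yellow borders, collapsed by default and offset above the button's top edge.
# The event graph is empty, so this asset supplies only layout and styling.
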WidgetBlueprint /FortniteCoreUI/Athena/HUD/Bacchus/HUDLayoutTool/HUDLayoutToolV2_LayoutButton
graphs [EventGraph {links [], nodes []}]
variables []
widgets {
 info {inheritedWidgetCount=3, parentClass=/Script/FortniteUI.HUDLayoutToolV2_LayoutButton, rootWidgetClass=&Overlay, widgetCount=7}
 widgets [
  Overlay_0 {widgetClassPath=&Overlay, widget=Overlay_0}
  Button_Internal {
   widgetClassPath=/Game/Athena/HUD/Bacchus/HUDLayoutTool/HUDLayoutToolV2_Button
   bInherited=true
   bIsVariable=true
   parent=Overlay_0
   widget=Button_Internal
   properties {
    bIsToggleable=true
    bOverride_HeightOverride=true
    bOverride_WidthOverride=true
    buttonHeightOverride=120
    buttonWidthOverride=530
    text=Layout name
    textFontSize=38
   }
   slot {class=OverlaySlot, properties {horizontalAlignment=HAlign_Fill, verticalAlignment=VAlign_Fill}}
  }
  Overlay_1 {
   widgetClassPath=&Overlay
   parent=Overlay_0
   widget=Overlay_1
   slot {class=OverlaySlot, properties {horizontalAlignment=HAlign_Fill, verticalAlignment=VAlign_Fill}}
  }
  Border_ActiveIndicator {
   widgetClassPath=&Border
   bInherited=true
   parent=Overlay_1
   widget=Border_ActiveIndicator
   properties {verticalAlignment=VAlign_Center, visibility=Collapsed, brushColor=<rgba 0.9301 1 0.0097 1>, padding=<ltrb 20 6 20 6>}
   slot {class=OverlaySlot, properties {horizontalAlignment=HAlign_Right, padding=<ltrb 0 -28 -12 0>}}
  }
  Text_ActiveIndicator {
   widgetClassPath=&CommonTextBlock
   parent=Border_ActiveIndicator
   widget=Text_ActiveIndicator
   properties {
    justification=Right
    scrollStyle=&TextScrollStyle-Base_C
    text=Active
    textTransformPolicy=ToUpper
    colorAndOpacity {colorUseRule=UseColor_Specified, specifiedColor=<rgba 0 0 0 1>}
    font {
     bForceMonospaced=false
     bMaterialIsStencil=false
     fontMaterial=null
     fontObject=&BurbankSmall
     letterSpacing=0
     monospacedWidth=1
     size=22
     skewAmount=0
     typefaceFontName=Black
     outlineSettings {
      bApplyOutlineToDropShadows=false
      bMiteredCorners=false
      bSeparateFillAlpha=false
      outlineMaterial=null
      outlineSize=0
      outlineColor=<rgba 0 0 0 1>
     }
    }
    shadowOffset=<0 0>
   }
   slot {class=BorderSlot, properties {verticalAlignment=VAlign_Center, padding=<ltrb 20 6 20 6>}}
  }
  Border_LegacyIndicator {
   widgetClassPath=&Border
   bInherited=true
   parent=Overlay_1
   widget=Border_LegacyIndicator
   properties {verticalAlignment=VAlign_Center, visibility=Collapsed, brushColor=<rgba 0.9301 1 0.0097 1>, padding=<ltrb 20 6 20 6>}
   slot {class=OverlaySlot, properties {padding=<ltrb -12 -28 0 0>}}
  }
  Text_LegacyIndicator {
   widgetClassPath=&CommonTextBlock
   parent=Border_LegacyIndicator
   widget=Text_LegacyIndicator
   properties {
    justification=Right
    scrollStyle=&TextScrollStyle-Base_C
    text=Legacy
    textTransformPolicy=ToUpper
    colorAndOpacity {colorUseRule=UseColor_Specified, specifiedColor=<rgba 0 0 0 1>}
    font {
     bForceMonospaced=false
     bMaterialIsStencil=false
     fontMaterial=null
     fontObject=&BurbankSmall
     letterSpacing=0
     monospacedWidth=1
     size=22
     skewAmount=0
     typefaceFontName=Black
     outlineSettings {
      bApplyOutlineToDropShadows=false
      bMiteredCorners=false
      bSeparateFillAlpha=false
      outlineMaterial=null
      outlineSize=0
      outlineColor=<rgba 0 0 0 1>
     }
    }
    shadowOffset=<0 0>
   }
   slot {class=BorderSlot, properties {verticalAlignment=VAlign_Center, padding=<ltrb 20 6 20 6>}}
  }
 ]
}
