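# Tooltip widget for the HUD Layout Tool V2, with default text warning that the maximum instances are reached.
# A minimum-sized SizeBox holds an overlay of a faint, offset shadow image and a dark blue border panel, both drawn with the same tooltip material.
# The panel centers a wrapped description text block inherited from the C++ HUDLayoutToolV2_Tooltip parent.
# The event graph holds only unconnected stub events, so the Blueprint supplies just layout and styling.
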WidgetBlueprint /FortniteCoreUI/Athena/HUD/Bacchus/HUDLayoutTool/HUDLayoutToolV2_Tooltip
graphs [EventGraph {nodes [
   Event PreConstruct()
   Event Construct()
   Event Tick()
  ]}]
variables []
widgets {
 info {inheritedWidgetCount=1, parentClass=/Script/FortniteUI.HUDLayoutToolV2_Tooltip, rootWidgetClass=&SizeBox, widgetCount=5}
 widgets [
  SizeBox_0 {
   widgetClassPath=&SizeBox
   widget=SizeBox_0
   properties {bOverride_MinDesiredHeight=true, bOverride_MinDesiredWidth=true, minDesiredHeight=140, minDesiredWidth=360}
  }
  Overlay {widgetClassPath=&Overlay, parent=SizeBox_0, widget=Overlay, slot {class=SizeBoxSlot}}
  Image_Shadow {
   widgetClassPath=&Image
   parent=Overlay
   widget=Image_Shadow
   properties {
    visibility=SelfHitTestInvisible
    brush {
     bIsDynamicallyLoaded=false
     drawAs=Box
     imageType=NoImage
     mirroring=NoMirror
     resourceName=None
     resourceObject=/Game/Athena/HUD/Bacchus/HUDLayoutTool/HLT_Icons/MI_HUDLayoutToolV2_Tooltip
     tiling=NoTile
     imageSize=<64 64>
     margin=<ltrb 0 0 0 0>
     outlineSettings {
      bUseBrushTransparency=false
      roundingType=HalfHeightRadius
      width=0
      color {colorUseRule=UseColor_Specified, specifiedColor=<rgba 0 0 0 0>}
      cornerRadii=<0 0 0 0>
     }
     tintColor {colorUseRule=UseColor_Specified, specifiedColor=<rgba 1 1 1 1>}
     uVRegion {bIsValid=false, max=<0 0>, min=<0 0>}
    }
    colorAndOpacity=<rgba 0 0 0 0.2>
    renderTransform {angle=0, scale=<1 1>, shear=<0 0>, translation=<8 8>}
   }
   slot {class=OverlaySlot, properties {horizontalAlignment=HAlign_Fill, verticalAlignment=VAlign_Fill}}
  }
  Border_Panel {
   widgetClassPath=&Border
   parent=Overlay
   widget=Border_Panel
   properties {
    horizontalAlignment=HAlign_Center
    verticalAlignment=VAlign_Center
    background {
     bIsDynamicallyLoaded=false
     drawAs=Image
     imageType=NoImage
     mirroring=NoMirror
     resourceName=None
     resourceObject=&MI_HUDLayoutToolV2_Tooltip
     tiling=NoTile
     imageSize=<32 32>
     margin=<ltrb 0 0 0 0>
     outlineSettings {
      bUseBrushTransparency=false
      roundingType=HalfHeightRadius
      width=0
      color {colorUseRule=UseColor_Specified, specifiedColor=<rgba 0 0 0 0>}
      cornerRadii=<0 0 0 0>
     }
     tintColor {colorUseRule=UseColor_Specified, specifiedColor=<rgba 1 1 1 1>}
     uVRegion {bIsValid=false, max=<0 0>, min=<0 0>}
    }
    brushColor=<rgba 0.0075 0.0595 0.2961 1>
    padding=<ltrb 20 18 20 20>
   }
   slot {class=OverlaySlot, properties {horizontalAlignment=HAlign_Fill, verticalAlignment=VAlign_Fill}}
  }
  TextBlock_Description {
   widgetClassPath=&CommonTextBlock
   bInherited=true
   parent=Border_Panel
   widget=TextBlock_Description
   properties {
    lineHeightPercentage=0.9
    text=Maximum instances reached. Remove an instance to be able to add another one.
    wrapTextAt=340
    font {
     bForceMonospaced=false
     bMaterialIsStencil=false
     fontMaterial=null
     fontObject=&BurbankSmall
     letterSpacing=0
     monospacedWidth=1
     size=24
     skewAmount=0
     typefaceFontName=Bold
     outlineSettings {
      bApplyOutlineToDropShadows=false
      bMiteredCorners=false
      bSeparateFillAlpha=false
      outlineMaterial=null
      outlineSize=0
      outlineColor=<rgba 0 0 0 1>
     }
    }
    shadowOffset=<0 0>
   }
   slot {class=BorderSlot, properties {horizontalAlignment=HAlign_Center, verticalAlignment=VAlign_Center, padding=<ltrb 20 18 20 20>}}
  }
 ]
}
