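# Widget Blueprint for a widget preview in the HUD Layout Tool V2, built on the native HUDLayoutToolV2_WidgetPreview class.
# A fixed-size square SizeBox holds an Overlay with a tinted fill image and a bottom-right text counter for instances left.
# The counter binds to an inherited native widget, while the inherited Border_Background slot has no widget in this tree.
# The event graph is empty, so any behavior comes from the parent class.
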WidgetBlueprint /FortniteCoreUI/Athena/HUD/Bacchus/HUDLayoutTool/HUDLayoutToolV2_WidgetPreview
graphs [EventGraph {links [], nodes []}]
variables []
widgets {
 info {inheritedWidgetCount=2, parentClass=/Script/FortniteUI.HUDLayoutToolV2_WidgetPreview, rootWidgetClass=&SizeBox, widgetCount=5}
 widgets [
  SizeBox_32 {
   widgetClassPath=&SizeBox
   widget=SizeBox_32
   properties {
    bOverride_HeightOverride=true
    bOverride_MaxAspectRatio=true
    bOverride_MinAspectRatio=true
    bOverride_WidthOverride=true
    heightOverride=100
    widthOverride=100
   }
  }
  Overlay_0 {widgetClassPath=&Overlay, parent=SizeBox_32, widget=Overlay_0, slot {class=SizeBoxSlot}}
  Preview {
   widgetClassPath=&Image
   parent=Overlay_0
   widget=Preview
   properties {visibility=SelfHitTestInvisible, colorAndOpacity=<rgba 0.2657 0.286 1 1>}
   slot {class=OverlaySlot, properties {horizontalAlignment=HAlign_Fill, verticalAlignment=VAlign_Fill}}
  }
  TextBlock_InstancesLeftCounter {
   widgetClassPath=&CommonTextBlock
   bInherited=true
   parent=Overlay_0
   widget=TextBlock_InstancesLeftCounter
   properties {
    justification=Center
    scrollStyle=&TextScrollStyle-Base_C
    textTransformPolicy=ToUpper
    font {
     bForceMonospaced=false
     bMaterialIsStencil=false
     fontMaterial=null
     fontObject=&BurbankSmall
     letterSpacing=0
     monospacedWidth=1
     size=16
     skewAmount=0
     typefaceFontName=Black
     outlineSettings {
      bApplyOutlineToDropShadows=false
      bMiteredCorners=false
      bSeparateFillAlpha=false
      outlineMaterial=null
      outlineSize=0
      outlineColor=<rgba 0 0 0 1>
     }
    }
    shadowOffset=<0 0>
   }
   slot {class=OverlaySlot, properties {horizontalAlignment=HAlign_Right, verticalAlignment=VAlign_Bottom}}
  }
  Border_Background {widgetClassPath=&Border, bInherited=true, widget=null, properties=~}
 ]
}
